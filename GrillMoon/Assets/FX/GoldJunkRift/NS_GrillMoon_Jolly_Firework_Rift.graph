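# Niagara system for a firework-style rift: an explosive burst with a crack that opens and closes over four seconds.
# Corona, flares, shockwave mesh, firework meshes, falling fragments and GPU sparks form the burst, mostly tinted by three user colors.
# The two-layer crack mesh uses stepped curves on dynamic material parameters to open, snap wide, then close.
# A light and wide flares share the crack's lifetime and flash as it snaps open.
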NiagaraSystem /GrillMoon/Assets/FX/GoldJunkRift/NS_GrillMoon_Jolly_Firework_Rift
dependencies {
 usedDataInterfaces [
  &NiagaraDataInterfaceCurve
  &NiagaraDataInterfaceVectorCurve
  &NiagaraDataInterfaceVector2DCurve
 ]
 usedDynamicInputs [&ConvertVectorToPosition]
 usedModules [&SystemState]
}
emitters {
 Corona {
  emitterUpdateScript [
   &EmitterState(Life Cycle Mode=::NewEnumerator1, Loop Behavior=::NewEnumerator1, Loop Duration=1, UseLoopDelay=true)
   &SpawnBurst_Instantaneous(Use Spawn Probability=false, UseLoopCountLimit=false)
  ]
  particleSpawnScript [&InitializeParticle(
    Color=$User.FXColorA
    Lifetime=0.6
    Sprite Size=<2000 2000>
    Sprite Size Mode=::NewEnumerator4
    UsePositionOffset=false
    Write Material Random=true
   )]
  particleUpdateScript [
   &ParticleState()
   &ScaleSpriteSize(
    Scale Sprite Size Mode=::NewEnumerator0
    Uniform Scale Factor=&FloatFromCurve(FloatCurve=DataInterface() {propertyValues {Curve {keys [{arriveTangent=1.4067, interpMode=RCIM_Cubic, time=0, value=0.1}, {interpMode=RCIM_Cubic, leaveTangent=0.2191, time=1, value=1}]}}})
   )
   &ScaleColor(
    Scale Alpha=&FloatFromCurve(FloatCurve=DataInterface() {propertyValues {Curve {keys [<at 0 0 RCIM_Cubic>,<at 0.25 1 RCIM_Cubic>,<at 1 0>]}}}, Scale Curve=1.5)
    Scale RGB=<2 2 2>
    ScaleA=true
    ScaleRGB=true
   )
   &DynamicMaterialParameters(
    Index 0 Param 1=&FloatFromCurve(FloatCurve=DataInterface() {propertyValues {Curve {keys [{arriveTangent=-2.4947, interpMode=RCIM_Cubic, time=0, value=1.5}, <at 1 0.2 RCIM_Cubic>]}}})
    Index 0 Param 2=0.4
    Index 0 Param 3=0.6
    Index 0 Param 4=1
    Module.Index0Param0Write=true
    Module.Index0Param1Write=true
    Module.Index0Param2Write=true
    Module.Index0Param3Write=true
   )
   &CameraOffset(Camera Offset Amount=500)
  ]
  rendererClasses [&NiagaraSpriteRendererProperties]
  renderers [{rendererClass=&NiagaraSpriteRendererProperties}]
 }
 CrackSpline001 {
  emitterUpdateScript [
   &EmitterState(Life Cycle Mode=::NewEnumerator1, Loop Behavior=::NewEnumerator1, Loop Duration=1, UseLoopDelay=true)
   &SpawnBurst_Instantaneous(Use Spawn Probability=false, UseLoopCountLimit=false)
   &SpawnBurst_Instantaneous(Spawn Group=1, Use Spawn Probability=false, UseLoopCountLimit=false) {module=SpawnBurst_Instantaneous001}
  ]
  particleSpawnScript [
   &InitializeParticle(
    Lifetime=4
    Mesh Scale=<12 12 9>
    Mesh Scale Mode=::NewEnumerator4
    Position Offset=<0 0 50>
    UsePositionOffset=true
    Write Material Random=true
   )
   &Color(Color=&MakeLinearColorFromVectorAndFloat(
     "Vector (RGB)"=&MergeHelper_VectorSwitch(0=<1 1 1>, 1=<2 2 2>, 2=?, 3=?, Index=&SpawnGroup(), Num=::NewEnumerator0)
     "Float (Alpha)"=&ClampFloat(Float=1)
    ))
   /Niagara/Modules/CascadeConversion/CascadeConversion_InitialMeshRotation(Rotation:Vector3f=<0 -0.5 0>)
   /SetVariables(Particles.VisibilityTag=&MergeHelper_IntSwitch(0=0, 1=1, 2=?, 3=?, 4=?, 5=?, Index=&SpawnGroup(), Num=::NewEnumerator0)) {isSetParameters=true}
  ]
  particleUpdateScript [
   &ParticleState()
   &ScaleMeshSize(Scale Factor=&VectorFromCurve(VectorCurve=DataInterface() {propertyValues {
       XCurve {keys [
         {arriveTangent=1.3847, interpMode=RCIM_Cubic, time=0, value=0.7}
         {arriveTangent=1.1111, interpMode=RCIM_Cubic, leaveTangent=1.1111, time=0.24, value=0.92}
         {arriveTangent=0.3077, interpMode=RCIM_Cubic, leaveTangent=0.3077, time=0.27, value=1}
         <at 0.5 1 RCIM_Cubic>
        ]}
       YCurve {keys [
         {arriveTangent=1.1207, interpMode=RCIM_Cubic, time=0, value=0.7}
         {arriveTangent=1.1111, interpMode=RCIM_Cubic, leaveTangent=1.1111, time=0.24, value=0.92}
         {arriveTangent=0.3077, interpMode=RCIM_Cubic, leaveTangent=0.3077, time=0.27, value=1}
         <at 0.5 1 RCIM_Cubic>
        ]}
       ZCurve {keys [
         <at 0 0.22 RCIM_Cubic>
         {arriveTangent=2.9709, interpMode=RCIM_Cubic, leaveTangent=2.9709, time=0.24, value=0.5}
         {arriveTangent=2.3077, interpMode=RCIM_Cubic, leaveTangent=2.3077, time=0.27, value=1.0221}
         <at 0.5 1.1 RCIM_Cubic>
        ]}
      }}))
   &DynamicMaterialParameters(
    Index 0 Param 1=&MergeHelper_FloatSwitch(
     0=&FloatFromCurve(FloatCurve=DataInterface() {propertyValues {Curve {keys [
          {arriveTangent=2.5415, interpMode=RCIM_Cubic, time=0, value=0}
          {arriveTangent=0.0898, leaveTangent=0.0898, time=0.0685, value=0.0885}
          {arriveTangent=0.0068, leaveTangent=0.0068, time=0.0876, value=0.1782}
          {arriveTangent=0.0991, leaveTangent=0.0991, time=0.1294, value=0.185}
          {arriveTangent=0.0029, leaveTangent=0.0029, time=0.1442, value=0.2841}
          {arriveTangent=0.113, leaveTangent=0.113, time=0.1714, value=0.287}
          {arriveTangent=7.2322, interpMode=RCIM_Cubic, leaveTangent=7.2322, time=0.24, value=0.4}
          <at 0.27 1>
          {arriveTangent=-0.3505, leaveTangent=-0.3505, time=0.7395, value=1}
          {arriveTangent=-0.0069, leaveTangent=-0.0069, time=0.7605, value=0.6495}
          {arriveTangent=-0.2804, leaveTangent=-0.2804, time=0.8367, value=0.6427}
          {arriveTangent=-0.0148, leaveTangent=-0.0148, time=0.8402, value=0.3623}
          {arriveTangent=-0.1832, leaveTangent=-0.1832, time=0.9216, value=0.3475}
          {arriveTangent=0.0012, leaveTangent=0.0012, time=0.9314, value=0.1644}
          {arriveTangent=-0.1656, leaveTangent=-0.1656, time=0.972, value=0.1656}
          <at 1 0 RCIM_Cubic>
         ]}}})
     1=&FloatFromCurve(FloatCurve=DataInterface() {propertyValues {Curve {keys [
          {arriveTangent=2.5415, interpMode=RCIM_Cubic, time=0, value=0}
          {arriveTangent=0.0898, leaveTangent=0.0898, time=0.0685, value=0.0885}
          {arriveTangent=0.0068, leaveTangent=0.0068, time=0.0876, value=0.1782}
          {arriveTangent=0.0991, leaveTangent=0.0991, time=0.1146, value=0.185}
          {arriveTangent=0.0029, leaveTangent=0.0029, time=0.1293, value=0.2841}
          {arriveTangent=0.113, leaveTangent=0.113, time=0.1694, value=0.287}
          {arriveTangent=8.8425, interpMode=RCIM_Cubic, leaveTangent=8.8425, time=0.2009, value=0.4}
          <at 0.25 1>
          {arriveTangent=-0.3505, leaveTangent=-0.3505, time=0.7395, value=1}
          {arriveTangent=-0.0069, leaveTangent=-0.0069, time=0.7605, value=0.6495}
          {arriveTangent=-0.2804, leaveTangent=-0.2804, time=0.8367, value=0.6427}
          {arriveTangent=-0.0148, leaveTangent=-0.0148, time=0.8402, value=0.3623}
          {arriveTangent=-0.1832, leaveTangent=-0.1832, time=0.9216, value=0.3475}
          {arriveTangent=0.0012, leaveTangent=0.0012, time=0.9314, value=0.1644}
          {arriveTangent=-0.1656, leaveTangent=-0.1656, time=0.972, value=0.1656}
          <at 1 0 RCIM_Cubic>
         ]}}})
     2=?
     3=?
     Index=&SpawnGroup()
     Num=::NewEnumerator0
    )
    Index 0 Param 2=&MergeHelper_FloatSwitch(
     0=&FloatFromCurve(FloatCurve=DataInterface() {propertyValues {Curve {keys [<at 0.4 2.5 RCIM_Cubic>, {arriveTangent=-7.514, interpMode=RCIM_Cubic, leaveTangent=-7.514, tangentMode=RCTM_User, time=1, value=0.05}]}}})
     1=&FloatFromCurve(FloatCurve=DataInterface() {propertyValues {Curve {keys [<at 0.4 2.5 RCIM_Cubic>, {arriveTangent=-8.5849, interpMode=RCIM_Cubic, leaveTangent=-8.5849, tangentMode=RCTM_User, time=1, value=0.05}]}}})
     2=?
     3=?
     Index=&SpawnGroup()
     Num=::NewEnumerator0
    )
    Index 0 Param 3=&MergeHelper_FloatSwitch(
     0=&FloatFromCurve(FloatCurve=DataInterface() {propertyValues {Curve {keys [
          {arriveTangent=4.7781, interpMode=RCIM_Cubic, leaveTangent=4.7781, tangentMode=RCTM_User, time=0, value=-0.1}
          {arriveTangent=0.2083, interpMode=RCIM_Cubic, leaveTangent=0.0033, tangentMode=RCTM_Break, time=0.0194, value=-0.05}
          {arriveTangent=-0.0015, interpMode=RCIM_Cubic, leaveTangent=2.7753, tangentMode=RCTM_Break, time=0.24, value=-0.05}
          {arriveTangent=0.1168, interpMode=RCIM_Cubic, leaveTangent=-0.0022, tangentMode=RCTM_Break, time=0.27, value=0.011}
          {arriveTangent=-0.0083, interpMode=RCIM_Cubic, leaveTangent=-0.0282, tangentMode=RCTM_Break, time=0.7539, value=0.01}
          {arriveTangent=-0.2604, interpMode=RCIM_Cubic, leaveTangent=-0.2604, time=0.7654, value=-0.003}
          {arriveTangent=-0.2775, interpMode=RCIM_Cubic, leaveTangent=-0.2775, time=0.8371, value=-0.0117}
          {arriveTangent=-0.2326, interpMode=RCIM_Cubic, leaveTangent=-0.2326, time=0.8389, value=-0.0234}
          {arriveTangent=-0.2151, interpMode=RCIM_Cubic, leaveTangent=-0.2151, time=0.9168, value=-0.0302}
          {arriveTangent=-0.2069, interpMode=RCIM_Cubic, leaveTangent=-0.2069, tangentMode=RCTM_User, time=0.9199, value=-0.0408}
          {arriveTangent=-0.5128, interpMode=RCIM_Cubic, leaveTangent=-0.5128, tangentMode=RCTM_User, time=1, value=-0.08}
         ]}}})
     1=&FloatFromCurve(FloatCurve=DataInterface() {propertyValues {Curve {keys [
          {arriveTangent=4.122, interpMode=RCIM_Cubic, leaveTangent=4.122, tangentMode=RCTM_User, time=0, value=-0.1}
          {arriveTangent=0.2494, interpMode=RCIM_Cubic, leaveTangent=-0.0032, tangentMode=RCTM_Break, time=0.0194, value=-0.05}
          {arriveTangent=-0.0051, interpMode=RCIM_Cubic, leaveTangent=1.3708, tangentMode=RCTM_Break, time=0.2004, value=-0.05}
          {arriveTangent=0.1084, interpMode=RCIM_Cubic, leaveTangent=-0.0026, tangentMode=RCTM_Break, time=0.25, value=0.011}
          {arriveTangent=0.0001, interpMode=RCIM_Cubic, leaveTangent=-0.0271, tangentMode=RCTM_Break, time=0.7539, value=0.01}
          {arriveTangent=-0.2604, interpMode=RCIM_Cubic, leaveTangent=-0.2604, time=0.7654, value=-0.003}
          {arriveTangent=-0.2775, interpMode=RCIM_Cubic, leaveTangent=-0.2775, time=0.8371, value=-0.0117}
          {arriveTangent=-0.2326, interpMode=RCIM_Cubic, leaveTangent=-0.2326, time=0.8389, value=-0.0234}
          {arriveTangent=-0.2151, interpMode=RCIM_Cubic, leaveTangent=-0.2151, time=0.9168, value=-0.0302}
          {arriveTangent=-0.0562, interpMode=RCIM_Cubic, leaveTangent=-0.0562, tangentMode=RCTM_User, time=0.9199, value=-0.0408}
          {arriveTangent=-0.5769, interpMode=RCIM_Cubic, leaveTangent=-0.5769, tangentMode=RCTM_User, time=1, value=-0.08}
         ]}}})
     2=?
     3=?
     Index=&SpawnGroup()
     Num=::NewEnumerator0
    )
    Index 0 Param 4=&MergeHelper_FloatSwitch(
     0=&FloatFromCurve(FloatCurve=DataInterface() {propertyValues {Curve {keys [
          {arriveTangent=52.796, interpMode=RCIM_Cubic, time=0.24, value=0}
          {arriveTangent=1.1429, interpMode=RCIM_Cubic, leaveTangent=0.1223, tangentMode=RCTM_Break, time=0.27, value=0.64}
          {arriveTangent=-0.0785, interpMode=RCIM_Cubic, leaveTangent=-0.0785, tangentMode=RCTM_User, time=0.8, value=0.64}
          {arriveTangent=-5.2178, interpMode=RCIM_Cubic, leaveTangent=-5.2178, tangentMode=RCTM_User, time=1, value=-0.25}
         ]}}})
     1=&FloatFromCurve(FloatCurve=DataInterface() {propertyValues {Curve {keys [
          {arriveTangent=151.5713, interpMode=RCIM_Cubic, time=0.24, value=0}
          {arriveTangent=1.1429, interpMode=RCIM_Cubic, leaveTangent=0.0098, tangentMode=RCTM_Break, time=0.25, value=0.64}
          {arriveTangent=0.0124, interpMode=RCIM_Cubic, leaveTangent=-1.1867, tangentMode=RCTM_Break, time=0.8, value=0.64}
          {arriveTangent=-6.7729, interpMode=RCIM_Cubic, leaveTangent=-6.7729, tangentMode=RCTM_User, time=1, value=-0.25}
         ]}}})
     2=?
     3=?
     Index=&SpawnGroup()
     Num=::NewEnumerator0
    )
    Module.Index0Param0Write=true
    Module.Index0Param1Write=true
    Module.Index0Param2Write=true
    Module.Index0Param3Write=true
   )
   &ScaleColor(
    Scale Alpha=&MergeHelper_FloatSwitch(
     0=&FloatFromCurve(FloatCurve=DataInterface() {propertyValues {Curve {keys [<at 0.95 1 RCIM_Cubic>,<at 1 0>]}}})
     1=&FloatFromCurve(FloatCurve=DataInterface() {propertyValues {Curve {keys [
          {arriveTangent=59.1921, interpMode=RCIM_Cubic, leaveTangent=59.1921, tangentMode=RCTM_User, time=0.23, value=0}
          {arriveTangentWeight=0.0067, interpMode=RCIM_Cubic, leaveTangentWeight=0.2173, tangentMode=RCTM_User, time=0.25, value=1}
          {arriveTangent=-1.3333, interpMode=RCIM_Cubic, leaveTangent=-1.3333, time=0.9, value=0.25}
          <at 1 0>
         ]}}})
     2=?
     3=?
     Index=&SpawnGroup()
     Num=::NewEnumerator0
    )
    Scale RGB=&MergeHelper_VectorSwitch(
     0=&VectorFromFloat(Value=&FloatFromCurve(FloatCurve=DataInterface() {propertyValues {Curve {keys [{arriveTangent=-40.0855, interpMode=RCIM_Cubic, leaveTangent=-40.0855, tangentMode=RCTM_User, time=0, value=5}, <at 0.15 1>]}}}))
     1=&VectorFromFloat(Value=&FloatFromCurve(FloatCurve=DataInterface() {propertyValues {Curve {keys [{arriveTangent=-46.7664, interpMode=RCIM_Cubic, leaveTangent=-46.7664, tangentMode=RCTM_User, time=0.23, value=8}, <at 0.5 1>]}}}))
     2=?
     3=?
     Index=&SpawnGroup()
     Num=::NewEnumerator0
    )
    ScaleA=true
    ScaleRGB=true
   )
  ]
  rendererClasses [&NiagaraMeshRendererProperties]
  renderers [
   {rendererClass=&NiagaraMeshRendererProperties, rendererIndex=0}
   {rendererClass=&NiagaraMeshRendererProperties, rendererIndex=1}
  ]
 }
 FireworkFlare {
  emitterUpdateScript [
   &EmitterState(Life Cycle Mode=::NewEnumerator1, Loop Behavior=::NewEnumerator1, Loop Duration=1, UseLoopDelay=true)
   &SpawnBurst_Instantaneous(Use Spawn Probability=false, UseLoopCountLimit=false)
  ]
  particleSpawnScript [&InitializeParticle(
    Color=&MultiplyLinearColor(
     Linear Color=$User.FXColorC
     Scale Factor=&MakeFloat4FromVectorAndFloat("Vector (RGB)"=&VectorFromFloat(Value=2), "Float (Alpha)"=1)
    )
    Lifetime=0.25
    Sprite Rotation Angle=&RandomRangeFloat(Minimum=0, Maximum=1, Evaluation Type=::NewEnumerator1)
    Sprite Rotation Mode=::NewEnumerator1
    Sprite Size=<1800 1800>
    Sprite Size Mode=::NewEnumerator4
    UsePositionOffset=false
    Write Material Random=true
   )]
  particleUpdateScript [
   &ParticleState()
   &ScaleSpriteSize(
    Scale Sprite Size Mode=::NewEnumerator0
    Uniform Scale Factor=&FloatFromCurve(FloatCurve=DataInterface() {propertyValues {Curve {keys [<at 0 1 RCIM_Cubic>, {interpMode=RCIM_Cubic, leaveTangent=-2.2952, time=1, value=0}]}}})
   )
   &ScaleColor(
    Scale Alpha=&FloatFromCurve(FloatCurve=DataInterface() {propertyValues {Curve {keys [<at 0 0.7 RCIM_Cubic>, {interpMode=RCIM_Cubic, leaveTangent=-1.3318, time=1, value=0}]}}})
    Scale RGB=&VectorFromFloat(Value=&FloatFromCurve(FloatCurve=DataInterface() {propertyValues {Curve {keys [
          <at 0 10 RCIM_Cubic>
          <at 0.05 1>
          {arriveTangent=22.7273, interpMode=RCIM_Cubic, leaveTangent=22.7273, time=0.24, value=1}
          <at 0.27 6 RCIM_Cubic>
          <at 0.4 1 RCIM_Cubic>
         ]}}}))
    ScaleA=true
    ScaleRGB=true
   )
   &DynamicMaterialParameters(
    Index 0 Param 1=1
    Index 0 Param 2=2
    Index 0 Param 3=1
    Index 0 Param 4=1
    Module.Index0Param0Write=true
    Module.Index0Param1Write=true
    Module.Index0Param2Write=true
    Module.Index0Param3Write=true
   )
   &CameraOffset(Camera Offset Amount=500)
  ]
  rendererClasses [&NiagaraSpriteRendererProperties]
  renderers [{rendererClass=&NiagaraSpriteRendererProperties}]
 }
 FireworkSmall001 {
  emitterUpdateScript [
   &EmitterState(Inactive Response=::NewEnumerator1, Life Cycle Mode=::NewEnumerator1, Loop Behavior=::NewEnumerator1, Loop Duration=1, UseLoopDelay=true)
   &SpawnBurst_Instantaneous(Spawn Count=2, Use Spawn Probability=false, UseLoopCountLimit=false)
   &SpawnBurst_Instantaneous(Spawn Group=1, Use Spawn Probability=false, UseLoopCountLimit=false) {module=SpawnBurst_Instantaneous001}
  ]
  particleSpawnScript [
   &InitializeParticle(
    Color=&MultiplyLinearColor(
     Linear Color=$User.FXColorB
     Scale Factor=&MakeFloat4FromVectorAndFloat("Vector (RGB)"=<1 1 1>, "Float (Alpha)"=&RandomRangeFloat(Minimum=0.8, Maximum=1))
    )
    Lifetime=&RandomRangeFloat(Minimum=1.5, Maximum=2, Evaluation Type=::NewEnumerator1)
    Mesh Scale=&MergeHelper_VectorSwitch(
     0=<0.45 0.45 0.45>
     1=&VectorFromFloat(Value=&RandomRangeFloat(Minimum=0.6, Maximum=0.82, Evaluation Type=::NewEnumerator1))
     2=?
     3=?
     Index=&SpawnGroup()
     Num=::NewEnumerator0
    )
    Mesh Scale Mode=::NewEnumerator4
    UsePositionOffset=false
    Write Material Random=true
   )
   &CascadeConversion_InitialMeshRotation(Rotation:Vector3f=&RandomRangeVector(Minimum=<0 0 0>, Maximum=<1 1 1>, Evaluation Type=::NewEnumerator1))
  ]
  particleUpdateScript [
   &ParticleState()
   &ScaleMeshSize(Scale Factor=&VectorFromFloat(Value=&FloatFromCurve(FloatCurve=DataInterface() {propertyValues {Curve {keys [
          {arriveTangent=5.3978, interpMode=RCIM_Cubic, leaveTangent=5.3978, tangentMode=RCTM_User, time=0.0016, value=0.26}
          {arriveTangent=1.7427, interpMode=RCIM_Cubic, leaveTangent=1.7427, time=0.2631, value=1.4393}
          {interpMode=RCIM_Cubic, leaveTangent=0.2783, time=1, value=2}
         ]}}})))
   &DynamicMaterialParameters(
    Index 0 Param 2=&FloatFromCurve(FloatCurve=DataInterface() {propertyValues {Curve {keys [<at 0 0>,<at 1 1>]}}})
    Index 0 Param 3=&FloatFromCurve(FloatCurve=DataInterface() {propertyValues {Curve {keys [
         {arriveTangent=40.3185, interpMode=RCIM_Cubic, leaveTangent=40.3185, tangentMode=RCTM_User, time=0, value=0}
         {arriveTangent=5.8206, interpMode=RCIM_Cubic, leaveTangent=5.8206, tangentMode=RCTM_User, time=0.0999, value=2.9237}
         <at 1 5 RCIM_Cubic>
        ]}}})
    Index 0 Param 4=&FloatFromCurve(FloatCurve=DataInterface() {propertyValues {Curve {keys [
         <at -0.0044 0.5 RCIM_Cubic>
         {arriveTangentWeight=0.0909, interpMode=RCIM_Cubic, leaveTangentWeight=0.2831, tangentMode=RCTM_User, time=0.2396, value=1.5}
         <at 1 0 RCIM_Cubic>
        ]}}})
    Module.Index0Param0Write=true
    Module.Index0Param1Write=true
    Module.Index0Param2Write=true
    Module.Index0Param3Write=true
   )
   &ScaleColor(
    Scale RGB=&MergeHelper_VectorSwitch(
     0=&VectorFromCurve(VectorCurve=DataInterface() {propertyValues {XCurve {keys [<at 0 0.2093>,<at 1 0.02>]}, YCurve {keys [<at 0 0.05>,<at 1 0.48>]}, ZCurve {keys [<at 0 1>,<at 1 1>]}}})
     1=<1 1 1>
     2=?
     3=?
     Index=&SpawnGroup()
     Num=::NewEnumerator0
    )
    ScaleA=true
    ScaleRGB=true
   ) {enabled=false}
   &ScaleColor(
    Scale Alpha=&FloatFromCurve(FloatCurve=DataInterface() {propertyValues {Curve {keys [
         <at 0 1 RCIM_Cubic>
         {arriveTangentWeight=0.1034, interpMode=RCIM_Cubic, leaveTangentWeight=0.2069, tangentMode=RCTM_User, time=0.24, value=0.41}
         {arriveTangentWeight=0.1884, interpMode=RCIM_Cubic, leaveTangentWeight=0.0706, tangentMode=RCTM_User, time=0.72, value=0.41}
         <at 0.9 0 RCIM_Cubic>
        ]}}})
    Scale RGB=&VectorFromFloat(Value=&FloatFromCurve(FloatCurve=DataInterface() {propertyValues {Curve {keys [
          <at 0.01 0.99 RCIM_Cubic>
          {arriveTangent=-1.7059, interpMode=RCIM_Cubic, leaveTangent=-1.7059, time=0.08, value=0.53}
          {arriveTangent=-0.5824, interpMode=RCIM_Cubic, leaveTangent=-0.5824, time=0.52, value=0.12}
          <at 0.99 0 RCIM_Cubic>
         ]}}}))
    ScaleA=true
    ScaleRGB=true
   ) {module=ScaleColor001}
   &ScaleColor(
    Scale Alpha=&FloatFromCurve(FloatCurve=DataInterface() {propertyValues {Curve {keys [<at 0 0.8>,<at 1 0>]}}})
    ScaleA=true
    ScaleRGB=false
   ) {module=ScaleColor002}
  ]
  rendererClasses [&NiagaraMeshRendererProperties]
  renderers [{rendererClass=&NiagaraMeshRendererProperties, rendererIndex=0}]
 }
 Frags1 {
  emitterUpdateScript [
   &EmitterState(Life Cycle Mode=::NewEnumerator1, Loop Behavior=::NewEnumerator1, Loop Duration=1, UseLoopDelay=true)
   &SpawnBurst_Instantaneous(Spawn Count=9, Spawn Time=0.06, Use Spawn Probability=false, UseLoopCountLimit=false) {module=SpawnBurst_Instantaneous001}
   &SpawnBurst_Instantaneous(Spawn Count=12, Spawn Time=0.04, Use Spawn Probability=false, UseLoopCountLimit=false)
  ]
  particleSpawnScript [
   /SetVariables(Particles.SpawnRotationVector=&RandomRangeVector(Minimum=<0.1 0.5 -0.1>, Maximum=<0.4 1 0.1>, Evaluation Type=::NewEnumerator1)) {isSetParameters=true}
   &InitializeParticle(
    Color=&MultiplyLinearColor(Linear Color=$User.FXColorA, Scale Factor=<3 3 3 1>)
    Lifetime=&RandomRangeFloat(Minimum=1, Maximum=2, Evaluation Type=::NewEnumerator1)
    Mesh Scale=&RandomRangeVector(Minimum=<5 4 10>, Maximum=<9 7 12>, Evaluation Type=::NewEnumerator1)
    Mesh Scale Mode=::NewEnumerator4
    UsePositionOffset=false
    Write Material Random=true
   )
   &ShapeLocation(
    MaskSpawnGroup=false
    Non Uniform Scale=<1 1 0>
    Override Seed=false
    Sphere Radius=&RandomRangeFloat(Minimum=50, Maximum=150, Evaluation Type=::NewEnumerator1)
    UseOwnerScale=true
   )
   &AddVelocity(
    Velocity=&Multiply_VectorByFloat(
     Vector=$Particles.ShapeLocation.ShapeVector
     Float=&RandomRangeFloat(Minimum=16, Maximum=30, Evaluation Type=::NewEnumerator1)
    )
    Velocity Speed Scale=1.5
   )
   &CascadeConversion_InitialMeshRotation(Rotation:Vector3f=&RandomRangeVector(Minimum=<0 0 0>, Maximum=<1 1 1>, Evaluation Type=::NewEnumerator1))
  ]
  particleUpdateScript [
   &ParticleState()
   &ScaleMeshSize(Scale Factor=&VectorFromCurve(VectorCurve=DataInterface() {propertyValues {
       XCurve {keys [<at 0.2 1 RCIM_Cubic>, {arriveTangent=-1.25, interpMode=RCIM_Cubic, leaveTangent=-1.25, time=0.7932, value=0.7509}, <at 1 0>]}
       YCurve {keys [<at 0.2 1 RCIM_Cubic>, {arriveTangent=-0.25, interpMode=RCIM_Cubic, leaveTangent=-0.25, time=0.7932, value=0.85}, <at 1 0.8>]}
       ZCurve {keys [<at 0.2 1 RCIM_Cubic>,<at 0.7932 1 RCIM_Cubic>,<at 1 1>]}
      }}))
   &UpdateMeshOrientation(
    Rotation Rate=&FloatFromCurve(FloatCurve=DataInterface() {propertyValues {Curve {keys [
         {interpMode=RCIM_Cubic, leaveTangent=-1.3256, tangentMode=RCTM_Break, time=0, value=1}
         {interpMode=RCIM_Cubic, tangentMode=RCTM_Break, time=1, value=0}
        ]}}})
    Rotation Vector=$Particles.SpawnRotationVector
   )
   &ScaleColor(
    Scale Alpha=&FloatFromCurve(FloatCurve=DataInterface() {propertyValues {Curve {keys [<at 0.7 1 RCIM_Cubic>,<at 1 0>]}}})
    Scale RGB=&VectorFromFloat(Value=&FloatFromCurve(FloatCurve=DataInterface() {propertyValues {Curve {keys [
          {arriveTangent=-742.6354, interpMode=RCIM_Cubic, leaveTangent=-742.6354, tangentMode=RCTM_User, time=0.1, value=50}
          {arriveTangent=-13.4051, interpMode=RCIM_Cubic, leaveTangent=-13.4051, tangentMode=RCTM_User, time=0.1961, value=7.5192}
          <at 1 1>
         ]}}}))
    ScaleA=true
    ScaleRGB=true
   )
   &Drag(Drag=&RandomRangeFloat(Minimum=1, Maximum=4), Use Linear Drag=true, Use Rotational Drag=false)
   &AccelerationForce(Acceleration=<0 0 -400>)
   &SolveForcesAndVelocity(Clamp Velocity=false, Limit Acceleration=false)
  ]
  rendererClasses [&NiagaraMeshRendererProperties]
  renderers [{rendererClass=&NiagaraMeshRendererProperties, rendererIndex=0}]
 }
 Light {
  emitterUpdateScript [
   &EmitterState(Life Cycle Mode=::NewEnumerator1, Loop Behavior=::NewEnumerator1, Loop Duration=1, UseLoopDelay=true)
   &SpawnBurst_Instantaneous(Use Spawn Probability=false, UseLoopCountLimit=false)
  ]
  particleSpawnScript [&InitializeParticle(
    Color=&MultiplyLinearColor(
     Linear Color=<rgba 1 0.55 0.4 1>
     Scale Factor=&MakeFloat4FromVectorAndFloat("Vector (RGB)"=&VectorFromFloat(Value=2), "Float (Alpha)"=1)
    )
    Lifetime=4
    Sprite Size=<3000 3000>
    Sprite Size Mode=::NewEnumerator4
    UsePositionOffset=false
    Write Material Random=true
   )]
  particleUpdateScript [
   /SetVariables(Particles.LightExponent=8) {isSetParameters=true}
   &ParticleState()
   &ScaleColor(
    Scale Alpha=&FloatFromCurve(FloatCurve=DataInterface() {propertyValues {Curve {keys [<at 0 0.7 RCIM_Cubic>, {interpMode=RCIM_Cubic, leaveTangent=-1.3318, time=1, value=0}]}}})
    Scale RGB=&VectorFromFloat(Value=&FloatFromCurve(FloatCurve=DataInterface() {propertyValues {Curve {keys [
          <at 0 2 RCIM_Cubic>
          <at 0.05 1>
          {arriveTangent=22.7273, interpMode=RCIM_Cubic, leaveTangent=22.7273, time=0.24, value=1}
          <at 0.27 6 RCIM_Cubic>
          <at 0.4 1 RCIM_Cubic>
         ]}}}))
    ScaleA=true
    ScaleRGB=true
   )
   &CascadeConversion_LightProperties(LightBrightnessScale=32, LightColorScale=<1 1 1>, LightRadiusScale=1)
  ]
  rendererClasses [&NiagaraLightRendererProperties]
  renderers [{rendererClass=&NiagaraLightRendererProperties, rendererIndex=0}]
 }
 ShockWave {
  emitterUpdateScript [
   &EmitterState(Life Cycle Mode=::NewEnumerator1, Loop Behavior=::NewEnumerator1, Loop Duration=1, UseLoopDelay=true)
   &SpawnBurst_Instantaneous(Use Spawn Probability=false, UseLoopCountLimit=false)
  ]
  particleSpawnScript [
   &InitializeParticle(Lifetime=0.6, Mesh Scale=<2 2 2>, Mesh Scale Mode=::NewEnumerator4, UsePositionOffset=false, Write Material Random=true)
   &Color(Color=&MakeLinearColorFromVectorAndFloat("Vector (RGB)"=<1 1 1>, "Float (Alpha)"=&ClampFloat(Float=0.1)))
  ]
  particleUpdateScript [
   &ParticleState()
   &ScaleMeshSize(Scale Factor=&VectorFromCurve(VectorCurve=DataInterface() {propertyValues {XCurve {keys [<at 0 0.2>,<at 1 8>]}, YCurve {keys [<at 0 0.2>,<at 1 8>]}, ZCurve {keys [<at 0 0.2>,<at 1 8>]}}}))
   &ScaleColor(
    Scale Alpha=&FloatFromCurve(FloatCurve=DataInterface() {propertyValues {Curve {keys [{arriveTangent=-2.1694, interpMode=RCIM_Cubic, time=0, value=1}, <at 1 0 RCIM_Cubic>]}}})
    ScaleA=true
    ScaleRGB=true
   )
   &CameraOffset(Camera Offset Amount=500)
  ]
  rendererClasses [&NiagaraMeshRendererProperties]
  renderers [{rendererClass=&NiagaraMeshRendererProperties, rendererIndex=0}]
 }
 SparkliesTop {
  simTarget=GPUComputeSim
  emitterUpdateScript [
   &EmitterState(Life Cycle Mode=::NewEnumerator1, Loop Behavior=::NewEnumerator1, Loop Duration=0.1, UseLoopDelay=true)
   &SpawnBurst_Instantaneous(Spawn Count=25, Spawn Time=0.04, Use Spawn Probability=false, UseLoopCountLimit=false) {module=SpawnBurst_Instantaneous001}
   &SpawnBurst_Instantaneous(Spawn Count=35, Use Spawn Probability=false, UseLoopCountLimit=false)
   &SpawnRate(SpawnRate=300, Use Spawn Probability=false)
  ]
  particleSpawnScript [
   /SetVariables(Particles.SpriteRotationRate=&RandomRangeFloat(Minimum=-1, Maximum=1, Evaluation Type=::NewEnumerator1)) {isSetParameters=true}
   &InitializeParticle(
    Color=&MultiplyLinearColor(
     Linear Color=$User.FXColorC
     Scale Factor=&MakeFloat4FromVectorAndFloat("Vector (RGB)"=&VectorFromFloat(Value=1.5), "Float (Alpha)"=1)
    )
    Lifetime=&RandomRangeFloat(Minimum=0.25, Maximum=1.6, Evaluation Type=::NewEnumerator1)
    Position Mode=::NewEnumerator0
    Sprite Rotation Angle=&RandomRangeFloat(Minimum=0, Maximum=1, Evaluation Type=::NewEnumerator1)
    Sprite Rotation Mode=::NewEnumerator1
    Sprite Size=&Vector2DFromFloat(Value=&RandomRangeFloat(Minimum=60, Maximum=88, Evaluation Type=::NewEnumerator1))
    Sprite Size Mode=::NewEnumerator4
    Write Material Random=true
   )
   &AddVelocity(Velocity=&RandomRangeVector(Minimum=<0 0 1000>, Maximum=<0 0 5000>, Evaluation Type=::NewEnumerator1))
   &ShapeLocation(
    Hemisphere Distribution=<1 0.5>
    MaskSpawnGroup=false
    Offset Mode=::NewEnumerator0
    Override Seed=false
    Sphere Radius=&RandomRangeFloat(Minimum=5, Maximum=28, Evaluation Type=::NewEnumerator1)
    UseOwnerScale=true
   )
   &AddVelocity(Velocity=&Multiply_VectorByFloat(
     Vector=$Particles.ShapeLocation.ShapeVector
     Float=&RandomRangeFloat(Minimum=0, Maximum=50, Evaluation Type=::NewEnumerator1)
    )) {module=AddVelocity001}
  ]
  particleUpdateScript [
   &ParticleState()
   &SpriteRotationRate(Rotation Rate=&AngleConversion(
     Angle:NiagaraFloat=$Particles.SpriteRotationRate
     Angle Input:ENiagara_AngleInput=::NewEnumerator1
     Angle Output:ENiagara_AngleInput=::NewEnumerator0
    ))
   &ScaleSpriteSize(
    Scale Sprite Size Mode=::NewEnumerator0
    Uniform Scale Factor=&FloatFromCurve(FloatCurve=DataInterface() {propertyValues {Curve {keys [
         {arriveTangent=-0.4739, time=-0.0044, value=1}
         {arriveTangent=0.2957, interpMode=RCIM_Cubic, leaveTangent=0.2957, time=0.157, value=0.5261}
         {arriveTangent=1.0007, interpMode=RCIM_Cubic, leaveTangent=1.0007, time=0.1686, value=1.0512}
         {arriveTangent=-3.1557, interpMode=RCIM_Cubic, leaveTangent=-3.1557, time=0.1876, value=0.5567}
         {arriveTangent=1.2911, interpMode=RCIM_Cubic, leaveTangent=1.2911, time=0.3841, value=0.3711}
         {arriveTangent=4.996, interpMode=RCIM_Cubic, leaveTangent=4.996, time=0.4102, value=0.8441}
         {arriveTangent=-1.6571, interpMode=RCIM_Cubic, leaveTangent=-1.6571, time=0.4265, value=0.583}
         {arriveTangent=0.0755, interpMode=RCIM_Cubic, leaveTangent=0.0755, time=0.69, value=0.3804}
         {arriveTangent=0.7225, interpMode=RCIM_Cubic, leaveTangent=0.7225, time=0.7007, value=0.6037}
         {arriveTangent=-2.017, interpMode=RCIM_Cubic, leaveTangent=-2.017, time=0.7133, value=0.3972}
         <at 1 0>
        ]}}})
   )
   &Drag(Drag=&RandomRangeFloat(Minimum=2, Maximum=4), Use Linear Drag=true, Use Rotational Drag=true)
   &SolveForcesAndVelocity(Clamp Velocity=false, Limit Acceleration=false)
   &ScaleColor(
    Scale RGB=&VectorFromFloat(Value=&FloatFromCurve(FloatCurve=DataInterface() {propertyValues {Curve {keys [<at 0 2 RCIM_Cubic>,<at 0.4 1>]}}}))
    ScaleA=true
    ScaleRGB=true
   )
  ]
  rendererClasses [&NiagaraSpriteRendererProperties]
  renderers [{rendererClass=&NiagaraSpriteRendererProperties}]
 }
 Sparks {
  simTarget=GPUComputeSim
  emitterUpdateScript [
   &EmitterState(Life Cycle Mode=::NewEnumerator1, Loop Behavior=::NewEnumerator1, Loop Duration=0.1, UseLoopDelay=true)
   &SpawnBurst_Instantaneous(Spawn Count=50, Use Spawn Probability=false, UseLoopCountLimit=false)
   &SpawnRate(SpawnRate=1400, Use Spawn Probability=false)
  ]
  particleSpawnScript [
   &InitializeParticle(
    Color=&MultiplyLinearColor(
     Linear Color=$User.FXColorC
     Scale Factor=&MakeFloat4FromVectorAndFloat("Vector (RGB)"=&VectorFromFloat(Value=5), "Float (Alpha)"=1)
    )
    Lifetime=&RandomRangeFloat(Minimum=0.4, Maximum=3, Evaluation Type=::NewEnumerator1)
    Sprite Size=&RandomRangeVector2D(Minimum=<4 4>, Maximum=<7 10>, Evaluation Type=::NewEnumerator1)
    Sprite Size Mode=::NewEnumerator4
    UsePositionOffset=false
    Write Material Random=true
   )
   &AddVelocity(Velocity=&RandomRangeVector(Minimum=<-300 0 0>, Maximum=<-150 0 0>, Evaluation Type=::NewEnumerator1))
   &ShapeLocation(
    MaskSpawnGroup=false
    Override Seed=false
    Sphere Radius=&RandomRangeFloat(Minimum=100, Maximum=120, Evaluation Type=::NewEnumerator1)
    UseOwnerScale=true
   )
   &AddVelocity(Velocity=&Multiply_VectorByFloat(Vector=$Particles.ShapeLocation.ShapeVector, Float=15)) {module=AddVelocity001}
  ]
  particleUpdateScript [
   &ParticleState()
   &ScaleSpriteSize(
    Scale Factor=&Vector2DFromCurve(
     Vector2Curve=DataInterface() {propertyValues {
       XCurve {keys [
         {arriveTangent=5.939, interpMode=RCIM_Cubic, time=0, value=0.5}
         {arriveTangent=-0.4, interpMode=RCIM_Cubic, leaveTangent=-0.4, time=0.15, value=1}
         <at 1 0.1>
        ]}
       YCurve {keys [
         {arriveTangent=28.4834, interpMode=RCIM_Cubic, time=0, value=0}
         {arriveTangent=0.15, interpMode=RCIM_Cubic, leaveTangent=0.15, time=0.15, value=1.5}
         <at 1 0.15>
        ]}
      }}
     ScaleCurve=<1 1>
    )
    Scale Sprite Size Mode=::NewEnumerator2
   )
   &AccelerationForce(Acceleration=<0 0 -800>)
   &Drag(Drag=&RandomRangeFloat(Minimum=1, Maximum=2), Use Linear Drag=true, Use Rotational Drag=true)
   &SolveForcesAndVelocity(Clamp Velocity=false, Limit Acceleration=false)
   &ScaleSpriteSizeBySpeed(Max Scale Factor=<1 2>, Source Velocity:Vector3f=&Multiply_Vector(A=<0.05 0.02 1>, B=$Particles.Velocity))
   &ScaleColor(
    Scale Alpha=&FloatFromCurve(FloatCurve=DataInterface() {propertyValues {Curve {keys [<at 0.7 1 RCIM_Cubic>,<at 1 0>]}}})
    Scale RGB=&VectorFromFloat(Value=&FloatFromCurve(FloatCurve=DataInterface() {propertyValues {Curve {keys [
          {interpMode=RCIM_Cubic, leaveTangent=-22.9178, tangentMode=RCTM_Break, time=0, value=5}
          {interpMode=RCIM_Cubic, tangentMode=RCTM_Break, time=0.35, value=1}
         ]}}}))
    ScaleA=true
    ScaleRGB=true
   )
  ]
  rendererClasses [&NiagaraSpriteRendererProperties]
  renderers [{rendererClass=&NiagaraSpriteRendererProperties}]
 }
 WideFlare001 {
  emitterUpdateScript [
   &EmitterState(Life Cycle Mode=::NewEnumerator1, Loop Behavior=::NewEnumerator1, Loop Duration=1, UseLoopDelay=true)
   &SpawnBurst_Instantaneous(Use Spawn Probability=false, UseLoopCountLimit=false)
   &SpawnBurst_Instantaneous(Spawn Group=1, Use Spawn Probability=false, UseLoopCountLimit=false) {module=SpawnBurst_Instantaneous001}
  ]
  particleSpawnScript [
   &InitializeParticle(
    Color=&MultiplyLinearColor(
     Linear Color=$User.FXColorC
     Scale Factor=&MakeFloat4FromVectorAndFloat("Vector (RGB)"=&VectorFromFloat(Value=2), "Float (Alpha)"=1)
    )
    Lifetime=4
    Sprite Size=&MergeHelper_Vector2Switch(0=<3000 150>, 1=<1400 1400>, 2=?, 3=?, 4=?, Index=&SpawnGroup(), Num=::NewEnumerator0)
    Sprite Size Mode=::NewEnumerator4
    UsePositionOffset=false
    Write Material Random=true
   )
   /SetVariables(Particles.VisibilityTag=&SpawnGroup()) {isSetParameters=true}
  ]
  particleUpdateScript [
   /SetVariables(Particles.Position=&AddVectorToPosition(
     Position=$Engine.Emitter.SimulationPosition
     Vector=&VectorFromCurve(VectorCurve=DataInterface() {propertyValues {
        XCurve {keys [<at 0.24 0 RCIM_Cubic>,<at 0.27 0>]}
        YCurve {keys [<at 0.24 0 RCIM_Cubic>,<at 0.27 0>]}
        ZCurve {keys [{arriveTangent=-6073.0132, interpMode=RCIM_Cubic, leaveTangent=-6073.0132, tangentMode=RCTM_User, time=0.24, value=0}, <at 0.27 -130>]}
       }})
    )) {isSetParameters=true}
   &ParticleState()
   &ScaleSpriteSize(
    Scale Factor=&MergeHelper_Vector2Switch(
     0=&Vector2DFromCurve(
      Vector2Curve=DataInterface() {propertyValues {XCurve {keys [<at 0 1 RCIM_Cubic>,<at 0.27 0.5>]}, YCurve {keys [<at 0 1 RCIM_Cubic>,<at 0.27 1>]}}}
      ScaleCurve=<1 1>
     )
     1=&Vector2DFromFloat(Value=&FloatFromCurve(FloatCurve=DataInterface() {propertyValues {Curve {keys [<at 0 1 RCIM_Cubic>, {interpMode=RCIM_Cubic, leaveTangent=-1.2932, time=1, value=0.2}]}}}))
     2=?
     3=?
     4=?
     Index=&SpawnGroup()
     Num=::NewEnumerator0
    )
    Scale Sprite Size Mode=::NewEnumerator2
   )
   &ScaleColor(
    Scale Alpha=&MergeHelper_FloatSwitch(
     0=&FloatFromCurve(FloatCurve=DataInterface() {propertyValues {Curve {keys [<at 0.0503 1 RCIM_Cubic>,<at 1 0>]}}})
     1=&FloatFromCurve(FloatCurve=DataInterface() {propertyValues {Curve {keys [<at 0 0.7 RCIM_Cubic>, {arriveTangent=-1.1087, interpMode=RCIM_Cubic, leaveTangent=-1.1087, tangentMode=RCTM_User, time=1, value=0}]}}})
     2=?
     3=?
     Index=&SpawnGroup()
     Num=::NewEnumerator0
    )
    Scale RGB=&MergeHelper_VectorSwitch(
     0=<5 5 5>
     1=&VectorFromFloat(Value=&FloatFromCurve(FloatCurve=DataInterface() {propertyValues {Curve {keys [
           <at 0 2 RCIM_Cubic>
           <at 0.05 1>
           {arriveTangent=242.3384, interpMode=RCIM_Cubic, leaveTangent=242.3384, tangentMode=RCTM_User, time=0.24, value=1}
           <at 0.27 6 RCIM_Cubic>
           <at 0.4 1 RCIM_Cubic>
          ]}}}))
     2=?
     3=?
     Index=&SpawnGroup()
     Num=::NewEnumerator0
    )
    ScaleA=true
    ScaleRGB=true
   )
   &DynamicMaterialParameters(
    Index 0 Param 1=&FloatFromCurve(FloatCurve=DataInterface() {propertyValues {Curve {keys [<at 0 1 RCIM_Cubic>,<at 0.8 0>]}}})
    Index 0 Param 2=2
    Index 0 Param 3=1
    Index 0 Param 4=1
    Module.Index0Param0Write=true
    Module.Index0Param1Write=true
    Module.Index0Param2Write=true
    Module.Index0Param3Write=true
   )
  ]
  rendererClasses [&NiagaraSpriteRendererProperties]
  renderers [
   {rendererClass=&NiagaraSpriteRendererProperties}
   {rendererClass=&NiagaraSpriteRendererProperties, rendererIndex=1}
  ]
 }
}
summary {
 systemName=NS_GrillMoon_Jolly_Firework_Rift
 emitters [
  {emitterName=Light, rendererClasses [&NiagaraLightRendererProperties]}
  {emitterName=ShockWave, rendererClasses [&NiagaraMeshRendererProperties]}
  {emitterName=Corona, rendererClasses [&NiagaraSpriteRendererProperties]}
  {emitterName=FireworkFlare, rendererClasses [&NiagaraSpriteRendererProperties]}
  {emitterName=Sparks, simTarget=GPUComputeSim, rendererClasses [&NiagaraSpriteRendererProperties]}
  {emitterName=Frags1, rendererClasses [&NiagaraMeshRendererProperties]}
  {emitterName=SparkliesTop, simTarget=GPUComputeSim, rendererClasses [&NiagaraSpriteRendererProperties]}
  {emitterName=CrackSpline001, rendererClasses [&NiagaraMeshRendererProperties]}
  {emitterName=WideFlare001, rendererClasses [&NiagaraSpriteRendererProperties]}
  {emitterName=FireworkSmall001, rendererClasses [&NiagaraMeshRendererProperties]}
 ]
 userVariables [
  User.FXColorA {type=LinearColor, description="", defaultValue {struct=LinearColor, value=<rgba 1 0.975 0.9 1>}}
  User.FXColorB {type=LinearColor, description="", defaultValue {struct=LinearColor, value=<rgba 1 0.25 0 1>}}
  User.FXColorC {type=LinearColor, description="", defaultValue {struct=LinearColor, value=<rgba 1 0.5314 0.0627 1>}}
 ]
}
