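# Niagara launch effect for "Jolly" in the GoldJunkRift FX set: a short charge-up that ends in an electric flash.
# Glow sprites climb on an accelerating position curve while three spinning rift shard meshes rise, then lightning meshes and a flickering light fire around half a second in.
# User FXColor parameters tint the layers, and curves animate scale, color and dynamic material parameters.
# Rocket, fuse, flare, sparks, smoke and residual arc emitters are disabled.
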NiagaraSystem /GrillMoon/Assets/FX/GoldJunkRift/NS_GrillMoon_Jolly_Launch
dependencies {
 usedDataInterfaces [
  &NiagaraDataInterfaceCurve
  &NiagaraDataInterfaceVectorCurve
  &NiagaraDataInterfaceVector2DCurve
  &NiagaraDataInterfaceSpriteRendererInfo
 ]
 usedDynamicInputs [&ConvertVectorToPosition]
 usedModules [&SystemState]
}
emitters {
 ElectricArcsResidual {
  bEnabled=false
  emitterUpdateScript [
   &EmitterState(Life Cycle Mode=::NewEnumerator1, Loop Behavior=::NewEnumerator1, Loop Delay=0.6, Loop Duration=0.15, UseLoopDelay=true)
   &SpawnRate(
    SpawnRate=&Multiply_Float(
     A=55
     B=&FloatFromCurve(FloatCurve=DataInterface() {propertyValues {Curve {keys [<at 0 1 RCIM_Cubic>,<at 1 0.5>]}}}, CurveIndex=$Emitter.NormalizedLoopAge)
    )
    Use Spawn Probability=false
   )
  ]
  particleSpawnScript [
   &InitializeParticle(
    Color=&MultiplyLinearColor(Linear Color=$User.FXColorA, Scale Factor=<22 22 22 1>)
    Lifetime=&RandomRangeFloat(Minimum=0.1, Maximum=0.2, Evaluation Type=::NewEnumerator1)
    Mesh Scale=&RandomRangeVector(Minimum=<2 1 1>, Maximum=<5 5 5>, Evaluation Type=::NewEnumerator1)
    Mesh Scale Mode=::NewEnumerator4
    Position Offset=&RandomRangeVector(Minimum=<0 0 50>, Maximum=<0 0 80>)
    UsePositionOffset=true
    Write Material Random=true
   )
   /Niagara/Modules/CascadeConversion/CascadeConversion_InitialMeshRotation(Rotation:Vector3f=&RandomRangeVector(Minimum=<0 0 0>, Maximum=<1 0 0>, Evaluation Type=::NewEnumerator1))
   &ShapeLocation(MaskSpawnGroup=false, Override Seed=false, Sphere Radius=3, UseOwnerScale=true)
   &AddVelocity(Velocity=&Multiply_VectorByFloat(Vector=$Particles.ShapeLocation.ShapeVector, Float=-30))
   &AddVelocity(Velocity=&RandomRangeVector(Minimum=<0 0 -200>, Maximum=<0 0 -50>, Evaluation Type=::NewEnumerator1)) {module=AddVelocity001}
  ]
  particleUpdateScript [
   &ParticleState()
   &ScaleMeshSize(Scale Factor=&VectorFromCurve(VectorCurve=DataInterface() {propertyValues {
       XCurve {keys [<at 0.0033 0.5 RCIM_Constant>,<at 0.2054 0.7619 RCIM_Constant>,<at 0.4101 0.8575 RCIM_Constant>,<at 0.6967 1>]}
       YCurve {keys [<at 0.0033 0.6004 RCIM_Constant>,<at 0.2054 0.6716 RCIM_Constant>,<at 0.4101 0.3646 RCIM_Constant>,<at 0.6967 0.2468>]}
       ZCurve {keys [<at 0.0033 0.7767 RCIM_Constant>,<at 0.2054 0.5215 RCIM_Constant>,<at 0.4101 0.4457 RCIM_Constant>,<at 0.6967 0.6282>]}
      }}))
   &DynamicMaterialParameters(
    Index 0 Param 1=&RandomRangeFloat(Minimum=2, Maximum=3)
    Index 0 Param 2=&FloatFromCurve(FloatCurve=DataInterface() {propertyValues {Curve {keys [{arriveTangent=-1.8894, interpMode=RCIM_Cubic, time=0, value=1}, <at 1 -0.25>]}}})
    Index 0 Param 3=&RandomRangeFloat(Minimum=0.2, Maximum=2)
    Index 0 Param 4=&RandomRangeFloat(Minimum=0, Maximum=1)
    Module.Index0Param0Write=true
    Module.Index0Param1Write=true
    Module.Index0Param2Write=true
    Module.Index0Param3Write=true
   ) {module=DynamicMaterialParameters001}
   &ScaleColor(
    Scale RGB=&VectorFromFloat(Value=&FloatFromCurve(FloatCurve=DataInterface() {propertyValues {Curve {keys [<at 0 50 RCIM_Cubic>, {arriveTangent=-45, interpMode=RCIM_Cubic, leaveTangent=-45, time=0.5, value=30}, <at 1 5>]}}}))
    ScaleA=true
    ScaleRGB=true
   )
   &Drag(Drag=2, Use Linear Drag=true, Use Rotational Drag=true)
   &SolveForcesAndVelocity(Clamp Velocity=false, Limit Acceleration=false)
  ]
  rendererClasses [&NiagaraMeshRendererProperties]
  renderers [{rendererClass=&NiagaraMeshRendererProperties, rendererIndex=0}]
 }
 Glow {
  emitterUpdateScript [
   &EmitterState(Life Cycle Mode=::NewEnumerator1, Loop Behavior=::NewEnumerator1, Loop Duration=1, UseLoopDelay=true)
   &SpawnBurst_Instantaneous(Use Spawn Probability=false, UseLoopCountLimit=false)
   &SpawnBurst_Instantaneous(Spawn Group=1, Use Spawn Probability=false, UseLoopCountLimit=false) {module=SpawnBurst_Instantaneous001}
  ]
  particleSpawnScript [
   &InitializeParticle(
    Color=&MergeHelper_ColorSwitch(0=<rgba 1 1 1 0.5>, 1=$User.FXColorA, 2=?, 3=?, Index=&SpawnGroup(), Num=::NewEnumerator0)
    Lifetime=0.6
    Sprite Size=&MergeHelper_Vector2Switch(0=<500 500>, 1=<700 400>, 2=?, 3=?, 4=?, Index=0, Num=::NewEnumerator0)
    Sprite Size Mode=::NewEnumerator4
    UsePositionOffset=false
    Write Material Random=true
   )
   /SetVariables(Particles.VisibilityTag=&SpawnGroup()) {isSetParameters=true}
  ]
  particleUpdateScript [
   /SetVariables(Particles.LightExponent=16) {isSetParameters=true}
   &ParticleState()
   &ScaleSpriteSize(
    Scale Sprite Size Mode=::NewEnumerator0
    Uniform Scale Factor=&MergeHelper_FloatSwitch(
     0=&FloatFromCurve(FloatCurve=DataInterface() {propertyValues {Curve {keys [
          {arriveTangent=1.217, interpMode=RCIM_Cubic, time=0, value=0.2}
          {arriveTangent=0.3, interpMode=RCIM_Cubic, leaveTangent=0.3, time=0.8595, value=1}
          {interpMode=RCIM_Cubic, leaveTangent=-5.8059, time=1, value=0.5}
         ]}}})
     1=&FloatFromCurve(FloatCurve=DataInterface() {propertyValues {Curve {keys [
          <at 0.25 0 RCIM_Cubic>
          {arriveTangent=2.6401, interpMode=RCIM_Cubic, leaveTangent=2.6401, time=0.7401, value=0.8995}
          {arriveTangentWeight=0.0778, interpMode=RCIM_Cubic, leaveTangentWeight=0.034, tangentMode=RCTM_User, time=0.8068, value=1.4701}
          {arriveTangent=-9.6995, interpMode=RCIM_Cubic, leaveTangent=-9.6995, tangentMode=RCTM_User, time=0.836, value=1.2212}
          {arriveTangent=-9.0173, interpMode=RCIM_Cubic, leaveTangent=-9.0173, tangentMode=RCTM_User, time=1, value=0}
         ]}}})
     2=?
     3=?
     Index=&SpawnGroup()
     Num=::NewEnumerator0
    )
   )
   &ScaleColor(
    Scale Alpha=&FloatFromCurve(FloatCurve=DataInterface() {propertyValues {Curve {keys [<at 0.8 1 RCIM_Cubic>,<at 1 0>]}}})
    Scale RGB=&VectorFromCurve(VectorCurve=DataInterface() {propertyValues {
       XCurve {keys [
         <at 0 1.5 RCIM_Cubic>
         {arriveTangent=-1.591, interpMode=RCIM_Cubic, leaveTangent=-1.591, time=0.7, value=0.0306}
         {arriveTangent=0.8434, interpMode=RCIM_Cubic, leaveTangent=0.8434, time=0.8, value=0.2272}
         <at 1 0.2837 RCIM_Cubic>
        ]}
       YCurve {keys [
         <at 0 0.1062 RCIM_Cubic>
         {arriveTangent=1.3082, interpMode=RCIM_Cubic, leaveTangent=1.3082, time=0.7, value=0.8602}
         {arriveTangent=2.4774, interpMode=RCIM_Cubic, leaveTangent=2.4774, time=0.8, value=1.1528}
         <at 1 1.6034 RCIM_Cubic>
        ]}
       ZCurve {keys [
         <at 0 1.3315 RCIM_Cubic>
         {arriveTangent=2.0857, interpMode=RCIM_Cubic, leaveTangent=2.0857, time=0.7, value=1}
         {arriveTangent=13.3333, interpMode=RCIM_Cubic, leaveTangent=13.3333, time=0.8, value=3}
         <at 1 5 RCIM_Cubic>
        ]}
      }})
    ScaleA=true
    ScaleRGB=true
   ) {enabled=false}
   &ScaleColor(
    Scale RGB=&MergeHelper_VectorSwitch(
     0=&MakeVectorFromLinearColorRGB(Color=&LerpLinearColors(
       StartColor=<rgba 1 1 1 1>
       EndColor=&MultiplyLinearColor(Linear Color=$User.FXColorA, Scale Factor=<3 3 3 1>)
       LerpFactor=&FloatFromCurve(FloatCurve=DataInterface() {propertyValues {Curve {keys [<at 0.7 0 RCIM_Cubic>,<at 1 1>]}}})
      ))
     1=&VectorFromFloat(Value=&FloatFromCurve(FloatCurve=DataInterface() {propertyValues {Curve {keys [{interpMode=RCIM_Cubic, tangentMode=RCTM_Break, time=0.7, value=1}, {interpMode=RCIM_Cubic, tangentMode=RCTM_Break, time=1, value=20}]}}}))
     2=?
     3=?
     Index=&SpawnGroup()
     Num=::NewEnumerator0
    )
    ScaleA=true
    ScaleRGB=true
   ) {module=ScaleColor001}
   &DynamicMaterialParameters(
    Index 0 Param 1=&FloatFromCurve(FloatCurve=DataInterface() {propertyValues {Curve {keys [<at 0 0.2>,<at 0.7 0.2>,<at 1 1>]}}})
    Index 0 Param 2=&FloatFromCurve(FloatCurve=DataInterface() {propertyValues {Curve {keys [{arriveTangent=1.6759, interpMode=RCIM_Cubic, time=0, value=0.25}, <at 0.6972 1.25 RCIM_Cubic>]}}})
    Index 0 Param 3=1
    Index 0 Param 4=1
    Module.Index0Param0Write=true
    Module.Index0Param1Write=true
    Module.Index0Param2Write=true
    Module.Index0Param3Write=true
   )
   &CameraOffset(Camera Offset Amount=&MergeHelper_FloatSwitch(0=50, 1=0, 2=?, 3=?, Index=&SpawnGroup(), Num=::NewEnumerator0))
   &CascadeConversion_LightProperties(LightBrightnessScale=50, LightColorScale=<1 1 1>, LightRadiusScale=1)
   /SetVariables(Particles.Position=&AddVectorToPosition(
     Position=$Engine.Emitter.SimulationPosition
     Vector=&VectorFromCurve(VectorCurve=DataInterface() {propertyValues {
        XCurve {keys [<at 0 0>,<at 0.85 0 RCIM_Cubic>,<at 1 0 RCIM_Cubic>]}
        YCurve {keys [<at 0 0>,<at 0.85 0 RCIM_Cubic>,<at 1 0 RCIM_Cubic>]}
        ZCurve {keys [
          {arriveTangent=41.7771, interpMode=RCIM_Cubic, leaveTangent=41.7771, tangentMode=RCTM_User, time=0, value=10}
          {arriveTangent=43.9961, interpMode=RCIM_Cubic, leaveTangent=43.9961, tangentMode=RCTM_User, time=0.85, value=45}
          {arriveTangent=1014.9672, interpMode=RCIM_Cubic, leaveTangent=1014.9672, tangentMode=RCTM_User, time=1, value=150}
         ]}
       }})
    )) {isSetParameters=true}
  ]
  rendererClasses [&NiagaraSpriteRendererProperties, &NiagaraLightRendererProperties]
  renderers [
   {rendererClass=&NiagaraSpriteRendererProperties}
   {rendererClass=&NiagaraLightRendererProperties, rendererIndex=1}
   {rendererClass=&NiagaraSpriteRendererProperties, rendererIndex=2}
  ]
 }
 Light {
  emitterUpdateScript [
   &EmitterState(Inactive Response=::NewEnumerator1, Life Cycle Mode=::NewEnumerator1, Loop Behavior=::NewEnumerator1, Loop Duration=1, UseLoopDelay=true)
   &SpawnBurst_Instantaneous(Spawn Time=0.52, Use Spawn Probability=false, UseLoopCountLimit=false)
  ]
  particleSpawnScript [
   &InitializeParticle(
    Color=&MultiplyLinearColor(Linear Color=$User.FXColorA, Scale Factor=<3 3 3 1>)
    Lifetime=0.5
    Position Offset=<0 0 200>
    Sprite Size=<1000 1000>
    Sprite Size Mode=::NewEnumerator4
    UsePositionOffset=true
    Write Material Random=true
   )
   &Color(Color=&MakeLinearColorFromVectorAndFloat("Vector (RGB)"=<0.4984 2.4582 3>, "Float (Alpha)"=&ClampFloat(Float=1))) {enabled=false}
   &AddVelocity(Velocity=<0 0 500>)
  ]
  particleUpdateScript [
   /SetVariables(Particles.LightExponent=12) {isSetParameters=true}
   &ParticleState()
   &ScaleColor(
    Scale Alpha=&FloatFromCurve(FloatCurve=DataInterface() {propertyValues {Curve {keys [
         {arriveTangent=-1.8703, interpMode=RCIM_Cubic, time=0, value=1}
         {arriveTangent=-2.1191, interpMode=RCIM_Cubic, leaveTangent=-2.1191, time=0.1596, value=0.0779}
         {arriveTangent=-0.3475, interpMode=RCIM_Cubic, leaveTangent=-0.3475, time=0.1746, value=0.6301}
         {arriveTangent=-4.2271, interpMode=RCIM_Cubic, leaveTangent=-4.2271, time=0.2606, value=0.0428}
         {arriveTangent=10.8949, interpMode=RCIM_Cubic, leaveTangent=10.8949, time=0.3126, value=0.0466}
         {arriveTangent=0.8318, interpMode=RCIM_Cubic, leaveTangent=0.8318, time=0.3467, value=0.9801}
         {arriveTangent=-2.3114, interpMode=RCIM_Cubic, leaveTangent=-2.3114, time=0.5009, value=0.2032}
         {arriveTangent=0.4621, interpMode=RCIM_Cubic, leaveTangent=0.4621, time=0.751, value=0.0454}
         {arriveTangent=0.3237, interpMode=RCIM_Cubic, leaveTangent=0.3237, time=0.7657, value=0.3256}
         {arriveTangent=-1.3893, interpMode=RCIM_Cubic, leaveTangent=-1.3893, time=0.8132, value=0.0655}
         <at 1 0 RCIM_Cubic>
        ]}}})
    Scale RGB=&VectorFromFloat(Value=&FloatFromCurve(FloatCurve=DataInterface() {propertyValues {Curve {keys [{arriveTangent=-33.8851, interpMode=RCIM_Cubic, time=0, value=5}, <at 0.4 0.5 RCIM_Cubic>]}}}))
    ScaleA=true
    ScaleRGB=true
   )
   &CascadeConversion_LightProperties(LightBrightnessScale=10, LightColorScale=<1 1 1>, LightRadiusScale=1)
   &SolveForcesAndVelocity(Clamp Velocity=false, Limit Acceleration=false)
  ]
  rendererClasses [&NiagaraLightRendererProperties]
  renderers [{rendererClass=&NiagaraLightRendererProperties, rendererIndex=0}]
 }
 Lightning001 {
  emitterUpdateScript [
   &EmitterState(Life Cycle Mode=::NewEnumerator1, Loop Behavior=::NewEnumerator1, Loop Delay=0.52, Loop Duration=0.5, UseLoopDelay=true)
   &SpawnBurst_Instantaneous(Use Spawn Probability=false, UseLoopCountLimit=false)
   &SpawnBurst_Instantaneous(Spawn Group=1, Spawn Time=0.04, Use Spawn Probability=false, UseLoopCountLimit=false) {module=SpawnBurst_Instantaneous001}
  ]
  particleSpawnScript [
   &InitializeParticle(
    Color=$User.FXColorD
    Lifetime=&RandomRangeFloat(Minimum=0.35, Maximum=0.45, Evaluation Type=::NewEnumerator1)
    Mesh Scale=&MergeHelper_VectorSwitch(
     0=&RandomRangeVector(Minimum=<1 0.8 0.8>, Maximum=<1 2 2>, Evaluation Type=::NewEnumerator1)
     1=&RandomRangeVector(Minimum=<1 0.8 0.8>, Maximum=<1 1 1>, Evaluation Type=::NewEnumerator1)
     2=?
     3=?
     Index=&SpawnGroup()
     Num=::NewEnumerator0
    )
    Mesh Scale Mode=::NewEnumerator4
    Position Offset=<0 0 35>
    Position Offset Coordinate Space=::Local
    UsePositionOffset=true
    Write Material Random=true
   )
   &AddVelocity(Velocity=&RandomRangeVector(Minimum=<0 0 15>, Maximum=<0 0 25>, Evaluation Type=::NewEnumerator1))
   &DynamicMaterialParameters(
    Index 0 Param 4=&RandomRangeFloat(Minimum=0, Maximum=1, Evaluation Type=::NewEnumerator1)
    Module.Index0Param0Write=false
    Module.Index0Param1Write=false
    Module.Index0Param2Write=false
    Module.Index0Param3Write=true
   ) {module=DynamicMaterialParameters001}
   &CascadeConversion_InitialMeshRotation(Rotation:Vector3f=&RandomRangeVector(Minimum=<0 0 0>, Maximum=<1 0 0>, Evaluation Type=::NewEnumerator1))
   &InitialMeshOrientation(Mesh Orientation Mode=::NewEnumerator2, Orientation Coordinate Space=::Local, Orientation Vector=<0 0 1>, Use Rotation Vector=false)
   /SetVariables(Particles.VisibilityTag=&SpawnGroup()) {isSetParameters=true}
  ]
  particleUpdateScript [
   &ParticleState()
   &ScaleMeshSize()
   &SolveForcesAndVelocity(Clamp Velocity=false, Limit Acceleration=false)
   &DynamicMaterialParameters(
    Index 0 Param 1=&FloatFromCurve(FloatCurve=DataInterface() {propertyValues {Curve {keys [
         {arriveTangent=0.6196, interpMode=RCIM_Cubic, leaveTangent=0.6196, tangentMode=RCTM_User, time=0, value=0}
         {arriveTangent=1.0147, interpMode=RCIM_Cubic, leaveTangent=1.0147, time=0.6, value=0.5}
         {arriveTangent=1.25, interpMode=RCIM_Cubic, leaveTangent=1.25, time=0.6047, value=0.6136}
         <at 1 1>
        ]}}})
    Index 0 Param 2=&FloatFromCurve(FloatCurve=DataInterface() {propertyValues {Curve {keys [
         {arriveTangent=0.6495, interpMode=RCIM_Cubic, leaveTangent=0.6495, tangentMode=RCTM_User, time=0, value=0.5}
         {arriveTangent=-0.4659, interpMode=RCIM_Cubic, leaveTangent=-0.4659, tangentMode=RCTM_User, time=0.5, value=0.5}
         {arriveTangent=-1, interpMode=RCIM_Cubic, leaveTangent=-1, time=0.6, value=1}
         <at 1 0>
        ]}}})
    Index 0 Param 3=&FloatFromCurve(FloatCurve=DataInterface() {propertyValues {Curve {keys [
         <at 0 0.25 RCIM_Cubic>
         {
          arriveTangent=0.9504
          arriveTangentWeight=0.5661
          interpMode=RCIM_Cubic
          leaveTangent=0.9504
          leaveTangentWeight=0.0506
          tangentMode=RCTM_User
          time=0.5508
          value=0.5
         }
         {arriveTangent=2.0063, interpMode=RCIM_Cubic, leaveTangent=2.0063, time=0.6, value=2}
         <at 0.8 1 RCIM_Cubic>
        ]}}})
    Module.Index0Param0Write=true
    Module.Index0Param1Write=true
    Module.Index0Param2Write=true
    Module.Index0Param3Write=false
   )
   &ScaleColor(
    Scale Alpha=&FloatFromCurve(FloatCurve=DataInterface() {propertyValues {Curve {keys [<at 0.25 1 RCIM_Cubic>,<at 0.7 0 RCIM_Cubic>]}}})
    Scale RGB=&VectorFromCurve(VectorCurve=DataInterface() {propertyValues {XCurve {keys [<at 0.4 0.75>,<at 0.6 1>]}, YCurve {keys [<at 0.4 0.75>,<at 0.6 1>]}, ZCurve {keys [<at 0.4 0.75>,<at 0.6 1>]}}})
    ScaleA=true
    ScaleRGB=true
   )
  ]
  rendererClasses [&NiagaraMeshRendererProperties]
  renderers [
   {rendererClass=&NiagaraMeshRendererProperties, rendererIndex=0}
   {rendererClass=&NiagaraMeshRendererProperties, rendererIndex=1}
  ]
 }
 RiftShards {
  emitterUpdateScript [
   &EmitterState(Life Cycle Mode=::NewEnumerator1, Loop Behavior=::NewEnumerator1, Loop Duration=1, UseLoopDelay=true)
   &SpawnBurst_Instantaneous(Spawn Count=3, Use Spawn Probability=false, UseLoopCountLimit=false)
  ]
  particleSpawnScript [
   /SetVariables(Particles.SpawnRotationVector=&RandomRangeVector(Minimum=<-0.3 -0.3 -0.3>, Maximum=<0.3 0.3 0.3>, Evaluation Type=::NewEnumerator1)) {isSetParameters=true}
   &InitializeParticle(
    Lifetime=&RandomRangeFloat(Minimum=0.6, Maximum=0.65, Evaluation Type=::NewEnumerator1)
    Mesh Scale=&RandomRangeVector(Minimum=<1 1 1>, Maximum=<1.6 1.6 1.6>, Evaluation Type=::NewEnumerator1)
    Mesh Scale Mode=::NewEnumerator4
    UsePositionOffset=false
    Write Material Random=true
   )
   &CascadeConversion_InitialMeshRotation(Rotation:Vector3f=&RandomRangeVector(Minimum=<0 0 0>, Maximum=<1 1 1>, Evaluation Type=::NewEnumerator1))
   &AddVelocity(Velocity=<0 0 220>)
  ]
  particleUpdateScript [
   &ParticleState()
   &UpdateMeshOrientation(
    Rotation Rate=&FloatFromCurve(FloatCurve=DataInterface() {propertyValues {Curve {keys [<at 0.3 1 RCIM_Cubic>,<at 0.8 0 RCIM_Cubic>]}}})
    Rotation Vector=$Particles.SpawnRotationVector
   )
   &ScaleMeshSize(Scale Factor=&VectorFromFloat(Value=&FloatFromCurve(FloatCurve=DataInterface() {propertyValues {Curve {keys [
          {arriveTangent=4.2202, interpMode=RCIM_Cubic, leaveTangent=4.2202, tangentMode=RCTM_User, time=0, value=0.2}
          {arriveTangent=1.5, interpMode=RCIM_Cubic, leaveTangent=1.5, time=0.1972, value=0.9}
          {arriveTangentWeight=0.2287, interpMode=RCIM_Cubic, leaveTangentWeight=0.1419, tangentMode=RCTM_User, time=0.6, value=1.1}
          {arriveTangent=-5.794, interpMode=RCIM_Cubic, leaveTangent=-5.794, tangentMode=RCTM_User, time=0.85, value=0}
         ]}}})))
   &DynamicMaterialParameters(
    Index 0 Param 1=&FloatFromCurve(FloatCurve=DataInterface() {propertyValues {Curve {keys [
         {arriveTangent=7.7337, interpMode=RCIM_Cubic, leaveTangent=7.7337, tangentMode=RCTM_User, time=0, value=-0.1}
         <at 0.5 1.2 RCIM_Cubic>
         <at 0.85 -0.1 RCIM_Cubic>
        ]}}})
    Index 0 Param 2=&FloatFromCurve(FloatCurve=DataInterface() {propertyValues {Curve {keys [
         {arriveTangent=-0.3836, interpMode=RCIM_Cubic, time=0.0018, value=0}
         {arriveTangent=1.1789, interpMode=RCIM_Cubic, leaveTangent=1.1789, time=0.5, value=-0.0643}
         <at 0.85 1 RCIM_Cubic>
        ]}}})
    Index 0 Param 3=&FloatFromCurve(FloatCurve=DataInterface() {propertyValues {Curve {keys [<at 0.3 0 RCIM_Cubic>,<at 0.7 1>]}}}, CurveIndex=$Emitter.LoopedAge)
    Index 0 Param 4=1
    Module.Index0Param0Write=true
    Module.Index0Param1Write=true
    Module.Index0Param2Write=true
    Module.Index0Param3Write=true
   )
   &ScaleColor(
    Scale RGB=&VectorFromCurve(VectorCurve=DataInterface() {propertyValues {
       XCurve {keys [{arriveTangent=4.058, interpMode=RCIM_Cubic, time=0.5, value=0.4039}, {interpMode=RCIM_Cubic, leaveTangent=11.4276, time=1, value=3.8111}]}
       YCurve {keys [{arriveTangent=5.3776, interpMode=RCIM_Cubic, time=0.5, value=0.2481}, {interpMode=RCIM_Cubic, leaveTangent=27.9784, time=1, value=9.3021}]}
       ZCurve {keys [{arriveTangent=7.1243, interpMode=RCIM_Cubic, time=0.5, value=0.5}, {interpMode=RCIM_Cubic, leaveTangent=29.8716, time=1, value=10}]}
      }})
    ScaleA=true
    ScaleRGB=true
   ) {enabled=false}
   &ScaleColor(
    Scale RGB=&VectorFromCurve(VectorCurve=DataInterface() {propertyValues {
       XCurve {keys [{arriveTangent=4.058, interpMode=RCIM_Cubic, time=0.5, value=0.4039}, {interpMode=RCIM_Cubic, leaveTangent=11.4276, time=1, value=3.8111}]}
       YCurve {keys [{arriveTangent=5.3776, interpMode=RCIM_Cubic, time=0.5, value=0.2481}, {interpMode=RCIM_Cubic, leaveTangent=27.9784, time=1, value=9.3021}]}
       ZCurve {keys [{arriveTangent=7.1243, interpMode=RCIM_Cubic, time=0.5, value=0.5}, {interpMode=RCIM_Cubic, leaveTangent=29.8716, time=1, value=10}]}
      }})
    ScaleA=true
    ScaleRGB=true
   ) {enabled=false, module=ScaleColor001}
   &ScaleColor(
    Scale RGB=&MakeVectorFromLinearColorRGB(Color=&LerpLinearColors(
      StartColor=<rgba 1 1 1 1>
      EndColor=&MultiplyLinearColor(Linear Color=$User.FXColorC, Scale Factor=<10 10 10 1>)
      LerpFactor=&FloatFromCurve(FloatCurve=DataInterface() {propertyValues {Curve {keys [<at 0.5 0>,<at 1 1>]}}})
     ))
    ScaleA=true
    ScaleRGB=true
   ) {module=ScaleColor002}
   &Drag(Drag=4, Use Linear Drag=true, Use Rotational Drag=true)
   &SolveForcesAndVelocity(Clamp Velocity=false, Limit Acceleration=false)
   &CameraOffset(Camera Offset Amount=25)
  ]
  rendererClasses [&NiagaraMeshRendererProperties]
  renderers [{rendererClass=&NiagaraMeshRendererProperties, rendererIndex=0}]
 }
 RisingFlare {
  bEnabled=false
  emitterUpdateScript [
   &EmitterState(Life Cycle Mode=::NewEnumerator1, Loop Behavior=::NewEnumerator1, Loop Duration=1, UseLoopDelay=true)
   &SpawnBurst_Instantaneous(Spawn Time=0.52, Use Spawn Probability=false, UseLoopCountLimit=false)
  ]
  particleSpawnScript [
   &InitializeParticle(
    Color=&MultiplyLinearColor(Linear Color=$User.FXColorA, Scale Factor=<8 8 8 1>)
    Lifetime=0.3
    Sprite Size=<200 400>
    Sprite Size Mode=::NewEnumerator4
    UsePositionOffset=false
    Write Material Random=true
   )
   &AddVelocity(Velocity=<0 0 1600>)
  ]
  particleUpdateScript [
   &ParticleState()
   &ScaleSpriteSize(
    Scale Factor=&Vector2DFromCurve(
     Vector2Curve=DataInterface() {propertyValues {
       XCurve {keys [
         {arriveTangent=11.3038, interpMode=RCIM_Cubic, leaveTangent=11.3038, tangentMode=RCTM_User, time=0, value=0}
         {arriveTangentWeight=0.067, interpMode=RCIM_Cubic, leaveTangentWeight=0.268, tangentMode=RCTM_User, time=0.2, value=2}
         <at 1 0.1>
        ]}
       YCurve {keys [
         {arriveTangent=11.2479, interpMode=RCIM_Cubic, leaveTangent=11.2479, tangentMode=RCTM_User, time=0, value=0}
         {
          arriveTangent=0.1195
          arriveTangentWeight=0.0943
          interpMode=RCIM_Cubic
          leaveTangent=0.1195
          leaveTangentWeight=0.3771
          tangentMode=RCTM_User
          time=0.2
          value=1.1
         }
         {arriveTangent=-0.085, tangentMode=RCTM_User, time=1, value=1}
        ]}
      }}
     ScaleCurve=<1 1>
    )
    Scale Sprite Size Mode=::NewEnumerator2
   )
   &SolveForcesAndVelocity(Clamp Velocity=false, Limit Acceleration=false)
   &DynamicMaterialParameters(
    Index 0 Param 1=&FloatFromCurve(FloatCurve=DataInterface() {propertyValues {Curve {keys [{arriveTangent=-1.9727, interpMode=RCIM_Cubic, leaveTangent=-1.9727, tangentMode=RCTM_User, time=0, value=0.8}, <at 1 0>]}}})
    Index 0 Param 2=2
    Index 0 Param 3=1
    Index 0 Param 4=1
    Module.Index0Param0Write=true
    Module.Index0Param1Write=true
    Module.Index0Param2Write=true
    Module.Index0Param3Write=true
   )
   &ScaleColor(
    Scale RGB=&VectorFromFloat(Value=&FloatFromCurve(FloatCurve=DataInterface() {propertyValues {Curve {keys [<at 0 1 RCIM_Cubic>, {arriveTangent=6.9308, interpMode=RCIM_Cubic, leaveTangent=6.9308, tangentMode=RCTM_User, time=1, value=5}]}}}))
    ScaleA=true
    ScaleRGB=true
   )
   &CameraOffset(Camera Offset Amount=50)
  ]
  rendererClasses [&NiagaraSpriteRendererProperties]
  renderers [{rendererClass=&NiagaraSpriteRendererProperties}]
 }
 Rocket {
  bEnabled=false
  emitterUpdateScript [
   &EmitterState(Life Cycle Mode=::NewEnumerator1, Loop Behavior=::NewEnumerator1, Loop Duration=1, UseLoopDelay=true)
   &SpawnBurst_Instantaneous(Use Spawn Probability=false, UseLoopCountLimit=false)
  ]
  particleSpawnScript [&InitializeParticle(Lifetime=0.6, Mesh Scale=<1.35 1.35 1.35>, Mesh Scale Mode=::NewEnumerator4, UsePositionOffset=false, Write Material Random=true)]
  particleUpdateScript [
   &ParticleState()
   &ScaleMeshSize(Scale Factor=&VectorFromCurve(VectorCurve=DataInterface() {propertyValues {
       XCurve {keys [
         <at 0.5 1 RCIM_Cubic>
         <at 0.759 1.1 RCIM_Cubic>
         {arriveTangent=-0.756, interpMode=RCIM_Cubic, leaveTangent=-0.756, tangentMode=RCTM_User, time=0.9442, value=1}
         {arriveTangent=-30.383, interpMode=RCIM_Cubic, leaveTangent=-30.383, tangentMode=RCTM_User, time=1, value=0}
        ]}
       YCurve {keys [
         <at 0.5 1 RCIM_Cubic>
         <at 0.759 1.1 RCIM_Cubic>
         {arriveTangent=-0.756, interpMode=RCIM_Cubic, leaveTangent=-0.756, tangentMode=RCTM_User, time=0.9442, value=1}
         {arriveTangent=-30.383, interpMode=RCIM_Cubic, leaveTangent=-30.383, tangentMode=RCTM_User, time=1, value=0}
        ]}
       ZCurve {keys [
         <at 0.5 0.9 RCIM_Cubic>
         {arriveTangent=-0.3768, interpMode=RCIM_Cubic, leaveTangent=-0.3768, tangentMode=RCTM_User, time=0.759, value=0.7037}
         {arriveTangent=5.3782, interpMode=RCIM_Cubic, leaveTangent=5.3782, time=0.9442, value=1.6081}
         {arriveTangent=6.4056, interpMode=RCIM_Cubic, leaveTangent=6.4056, tangentMode=RCTM_User, time=1, value=2}
        ]}
      }}))
   &DynamicMaterialParameters(
    Index 0 Param 1=&FloatFromCurve(FloatCurve=DataInterface() {propertyValues {Curve {keys [<at 0 0 RCIM_Cubic>,<at 0.7 2 RCIM_Cubic>,<at 1 0>]}}})
    Index 0 Param 2=1
    Index 0 Param 3=1
    Index 0 Param 4=1
    Module.Index0Param0Write=true
    Module.Index0Param1Write=true
    Module.Index0Param2Write=true
    Module.Index0Param3Write=true
   )
   &ScaleColor(
    Scale RGB=&VectorFromCurve(VectorCurve=DataInterface() {propertyValues {
       XCurve {keys [
         {arriveTangent=2.5, time=0, value=0.5}
         {arriveTangent=5.625, interpMode=RCIM_Cubic, leaveTangent=5.625, time=0.7, value=3}
         {arriveTangent=48.9525, interpMode=RCIM_Cubic, leaveTangent=48.9525, time=0.8, value=5}
         <at 1 17.6857 RCIM_Cubic>
        ]}
       YCurve {keys [
         {arriveTangent=2.5, time=0, value=0.5}
         {arriveTangent=7.7114, interpMode=RCIM_Cubic, leaveTangent=7.7114, time=0.7, value=3}
         {arriveTangent=8.4028, interpMode=RCIM_Cubic, leaveTangent=8.4028, time=0.8, value=6.6691}
         <at 1 5.5208 RCIM_Cubic>
        ]}
       ZCurve {keys [
         {arriveTangent=2.5, time=0, value=0.5}
         {arriveTangent=11.875, interpMode=RCIM_Cubic, leaveTangent=11.875, time=0.7, value=3}
         {arriveTangent=56.6667, interpMode=RCIM_Cubic, leaveTangent=56.6667, time=0.8, value=10}
         <at 1 20 RCIM_Cubic>
        ]}
      }})
    ScaleA=true
    ScaleRGB=true
   )
   /SetVariables(Particles.Position=&AddVectorToPosition(
     Position=$Engine.Emitter.SimulationPosition
     Vector=&VectorFromCurve(VectorCurve=DataInterface() {propertyValues {
        XCurve {keys [<at 0 0>,<at 0.85 0 RCIM_Cubic>,<at 1 0 RCIM_Cubic>]}
        YCurve {keys [<at 0 0>,<at 0.85 0 RCIM_Cubic>,<at 1 0 RCIM_Cubic>]}
        ZCurve {keys [
          {arriveTangent=35, time=0, value=10}
          {arriveTangent=140, interpMode=RCIM_Cubic, leaveTangent=140, time=0.85, value=45}
          {interpMode=RCIM_Cubic, leaveTangent=1894.8823, time=1, value=150}
         ]}
       }})
    )) {isSetParameters=true}
  ]
  rendererClasses [&NiagaraMeshRendererProperties]
  renderers [{rendererClass=&NiagaraMeshRendererProperties, rendererIndex=0}]
 }
 SparksClose {
  bEnabled=false
  simTarget=GPUComputeSim
  emitterUpdateScript [
   &EmitterState(Life Cycle Mode=::NewEnumerator1, Loop Behavior=::NewEnumerator1, Loop Duration=0.6, UseLoopDelay=true)
   &SpawnBurst_Instantaneous(Spawn Count=12, Spawn Time=0.58, Use Spawn Probability=false, UseLoopCountLimit=false) {module=SpawnBurst_Instantaneous001}
   &SpawnBurst_Instantaneous(Spawn Count=50, Spawn Time=0.54, Use Spawn Probability=false, UseLoopCountLimit=false)
   &SpawnRate(SpawnRate=250, Use Spawn Probability=false)
  ]
  particleSpawnScript [
   &InitializeParticle(
    Color=&MultiplyLinearColor(Linear Color=$User.FXColorB, Scale Factor=<5 5 5 1>)
    Lifetime=&RandomRangeFloat(Minimum=0.2, Maximum=0.7, Evaluation Type=::NewEnumerator1)
    Sprite Size=&RandomRangeVector2D(Minimum=<2 2>, Maximum=<3.5 3.5>, Evaluation Type=::NewEnumerator1)
    Sprite Size Mode=::NewEnumerator4
    UsePositionOffset=false
    Write Material Random=true
   )
   &AddVelocity(Velocity=&RandomRangeVector(Minimum=<0 0 -500>, Maximum=<0 0 -250>, Evaluation Type=::NewEnumerator1))
   &ShapeLocation(MaskSpawnGroup=false, Non Uniform Scale=<1 1 0>, Override Seed=false, Sphere Radius=4, UseOwnerScale=true)
   &AddVelocity(Velocity=&Multiply_VectorByFloat(
     Vector=$Particles.ShapeLocation.ShapeVector
     Float=&RandomRangeFloat(Minimum=25, Maximum=120, Evaluation Type=::NewEnumerator1)
    )) {module=AddVelocity001}
  ]
  particleUpdateScript [
   &ParticleState()
   &ScaleSpriteSize(
    Scale Factor=&Vector2DFromCurve(
     Vector2Curve=DataInterface() {propertyValues {
       XCurve {keys [
         {arriveTangent=5.0383, interpMode=RCIM_Cubic, leaveTangent=5.0383, tangentMode=RCTM_User, time=0, value=0.5}
         {
          arriveTangent=-1.0232
          arriveTangentWeight=0.0539
          interpMode=RCIM_Cubic
          leaveTangent=-1.0232
          leaveTangentWeight=0.3052
          tangentMode=RCTM_User
          time=0.15
          value=1
         }
         {arriveTangent=-0.0568, tangentMode=RCTM_User, time=1, value=0.1}
        ]}
       YCurve {keys [
         {arriveTangent=17.943, interpMode=RCIM_Cubic, leaveTangent=17.943, tangentMode=RCTM_User, time=0, value=0}
         {arriveTangent=-2.2809, interpMode=RCIM_Cubic, leaveTangent=-2.2809, tangentMode=RCTM_User, time=0.15, value=1.5}
         <at 1 0.15>
        ]}
      }}
     ScaleCurve=<1 1>
    )
    Scale Sprite Size Mode=::NewEnumerator2
   )
   &AccelerationForce(Acceleration=<0 0 -700>)
   &Drag(
    Drag=&RandomRangeFloat(Minimum=0, Maximum=0.4, Evaluation Type=::NewEnumerator1)
    Use Linear Drag=true
    Use Rotational Drag=true
   )
   &SolveForcesAndVelocity(Clamp Velocity=false, Limit Acceleration=false)
   &ScaleSpriteSizeBySpeed(
    Max Scale Factor=<1 4>
    Source Velocity:Vector3f=&Multiply_Vector(A=<0.05 0.02 1>, B=$Particles.Velocity)
    Velocity Threshold=500
   )
   &ScaleColor(
    Scale Alpha=&FloatFromCurve(FloatCurve=DataInterface() {propertyValues {Curve {keys [<at 0.7 1 RCIM_Cubic>,<at 1 0>]}}})
    Scale RGB=&VectorFromFloat(Value=&FloatFromCurve(FloatCurve=DataInterface() {propertyValues {Curve {keys [
          {interpMode=RCIM_Cubic, leaveTangent=-75.5106, tangentMode=RCTM_Break, time=0, value=10}
          {interpMode=RCIM_Cubic, tangentMode=RCTM_Break, time=0.2, value=1}
          <at 1 1>
         ]}}}))
    ScaleA=true
    ScaleRGB=true
   )
  ]
  rendererClasses [&NiagaraSpriteRendererProperties]
  renderers [{rendererClass=&NiagaraSpriteRendererProperties}]
 }
 TransSmoke {
  bEnabled=false
  emitterUpdateScript [
   &EmitterState(Life Cycle Mode=::NewEnumerator1, Loop Behavior=::NewEnumerator1, Loop Duration=0.5, UseLoopDelay=true)
   &SpawnBurst_Instantaneous(Spawn Count=16, Spawn Time=0.49, Use Spawn Probability=false, UseLoopCountLimit=false)
   &SpawnRate(
    SpawnRate=&Multiply_Float(
     A=110
     B=&FloatFromCurve(
      FloatCurve=DataInterface() {propertyValues {Curve {keys [<at 0 1 RCIM_Cubic>, {arriveTangent=-3.253, interpMode=RCIM_Cubic, leaveTangent=-3.253, tangentMode=RCTM_User, time=0.6, value=0}]}}}
      CurveIndex=$Emitter.LoopedAge
     )
    )
    Use Spawn Probability=false
   )
  ]
  particleSpawnScript [
   /SetVariables(Particles.SpriteRotationRate=&RandomRangeFloat(Minimum=-0.1, Maximum=0.1, Evaluation Type=::NewEnumerator1)) {isSetParameters=true}
   &InitializeParticle(
    Color=&RandomRangeLinearColor(Minimum=<rgba 1 1 1 0.2>, Maximum=<rgba 1 1 1 0.3>)
    Lifetime=&RandomRangeFloat(Minimum=0.35, Maximum=0.45, Evaluation Type=::NewEnumerator1)
    Position Offset=<0 0 15>
    Sprite Size=&Vector2DFromFloat(Value=&RandomRangeFloat(Minimum=64, Maximum=95, Evaluation Type=::NewEnumerator1))
    Sprite Size Mode=::NewEnumerator4
    UsePositionOffset=true
    Write Material Random=true
   )
   &ShapeLocation(
    Cylinder Height=10
    Cylinder Radius=&RandomRangeFloat(Minimum=15, Maximum=22)
    MaskSpawnGroup=false
    Override Seed=false
    Shape Primitive=::NewEnumerator1
    Surface Only=true
    Use Endcaps In Surface Only Mode=false
    UseOwnerScale=true
   ) {module=ShapeLocation001}
   &AddVelocity(Velocity=&Multiply_VectorByFloat(
     Vector=&Multiply_Vector(A=$Particles.ShapeLocation001.ShapeVector, B=<1 1 0>)
     Float=&FloatFromCurve(FloatCurve=DataInterface() {propertyValues {Curve {keys [<at 0 25 RCIM_Cubic>,<at 0.5 50>]}}}, CurveIndex=$Emitter.Age)
    ))
   &AddVelocity(Velocity=&VectorFromCurve(VectorCurve=DataInterface() {propertyValues {XCurve {keys [<at 0 0>,<at 0.5 0>]}, YCurve {keys [<at 0 0>,<at 0.5 0>]}, ZCurve {keys [<at 0 100>,<at 0.5 -50>]}}})) {module=AddVelocity001}
   &InheritVelocity(Inherited Velocity Amount Scale=<1 1 0.2>, Inherited Velocity Speed Limit=0, Speed Limit=false, UseSpeedThreshold=false)
  ]
  particleUpdateScript [
   &ParticleState()
   &SpriteRotationRate(Rotation Rate=&AngleConversion(
     Angle:NiagaraFloat=$Particles.SpriteRotationRate
     Angle Input:ENiagara_AngleInput=::NewEnumerator1
     Angle Output:ENiagara_AngleInput=::NewEnumerator0
    ))
   &ScaleSpriteSize(
    Scale Sprite Size Mode=::NewEnumerator0
    Uniform Scale Factor=&FloatFromCurve(FloatCurve=DataInterface() {propertyValues {Curve {keys [
         {arriveTangent=4.5398, interpMode=RCIM_Cubic, leaveTangent=4.5398, tangentMode=RCTM_User, time=0, value=0.0414}
         {arriveTangent=0.9586, interpMode=RCIM_Cubic, leaveTangent=0.9586, time=0.1999, value=0.7634}
         <at 1 1>
        ]}}})
   )
   &ScaleColor(
    Scale Alpha=&FloatFromCurve(FloatCurve=DataInterface() {propertyValues {Curve {keys [<at 0.6 1 RCIM_Cubic>, {arriveTangent=-4.3893, interpMode=RCIM_Cubic, leaveTangent=-4.3893, tangentMode=RCTM_User, time=1, value=0}]}}})
    Scale RGB=&MakeVectorFromLinearColorRGB(Color=&LerpLinearColors(
      StartColor=&MultiplyLinearColor(Linear Color=$User.FXColorB, Scale Factor=<7 7 7 1>)
      EndColor=<rgba 1 1 1 1>
      LerpFactor=&FloatFromCurve(FloatCurve=DataInterface() {propertyValues {Curve {keys [
           {arriveTangent=3.6624, interpMode=RCIM_Cubic, leaveTangent=3.6624, tangentMode=RCTM_User, time=0, value=0}
           {interpMode=RCIM_Cubic, tangentMode=RCTM_Break, time=0.5, value=1}
          ]}}})
     ))
    ScaleA=true
    ScaleRGB=true
   )
   &DynamicMaterialParameters(
    Index 0 Param 1=&FloatFromCurve(FloatCurve=DataInterface() {propertyValues {Curve {keys [<at 0.3 0.5 RCIM_Cubic>,<at 1 -1>]}}})
    Index 0 Param 3=&FloatFromCurve(FloatCurve=DataInterface() {propertyValues {Curve {keys [<at 0.5 0>,<at 1 0>]}}})
    Index 0 Param 4=&FloatFromCurve(FloatCurve=DataInterface() {propertyValues {Curve {keys [<at 0.6 1 RCIM_Cubic>,<at 1 0>]}}})
    Module.Index0Param0Write=true
    Module.Index0Param1Write=true
    Module.Index0Param2Write=true
    Module.Index0Param3Write=true
   )
   &Drag(Drag=4, Use Linear Drag=true, Use Rotational Drag=true)
   &SolveForcesAndVelocity(Clamp Velocity=false, Limit Acceleration=false)
   &CameraOffset(Camera Offset Amount=50)
   &SubUVAnimation(
    "Sprite Renderer "=DataInterface() {propertyValues {SpriteRenderer=&self:TransSmoke_7.NiagaraSpriteRendererProperties_19}}
    SubUV Lookup Index Scale=0.0667
    UseEndFrame=false
    UseStartFrame=false
   ) {module=SubUVAnimation001}
  ]
  rendererClasses [&NiagaraSpriteRendererProperties]
  renderers [{rendererClass=&NiagaraSpriteRendererProperties}]
 }
 fuse {
  bEnabled=false
  emitterUpdateScript [
   &EmitterState(Life Cycle Mode=::NewEnumerator1, Loop Behavior=::NewEnumerator1, Loop Duration=0.5, UseLoopDelay=true)
   &SpawnBurst_Instantaneous(Use Spawn Probability=false, UseLoopCountLimit=false)
   &SpawnRate(SpawnRate=30, Use Spawn Probability=false)
   &SpawnRate(SpawnGroup=1, SpawnRate=40, Use Spawn Probability=false) {module=SpawnRate001}
  ]
  particleSpawnScript [
   /SetVariables(Particles.SpriteRotationRate=&RandomRangeFloat(Minimum=0, Maximum=0.01, Evaluation Type=::NewEnumerator1)) {isSetParameters=true}
   &InitializeParticle(
    Color=&MergeHelper_ColorSwitch(
     0=&MultiplyLinearColor(Linear Color=$User.FXColorB, Scale Factor=<100 100 100 1>)
     1=&MultiplyLinearColor(Linear Color=$User.FXColorB, Scale Factor=<70 70 70 1>)
     2=?
     3=?
     Index=&SpawnGroup()
     Num=::NewEnumerator0
    )
    Lifetime=&MergeHelper_FloatSwitch(
     0=&RandomRangeFloat(Minimum=0.05, Maximum=0.1, Evaluation Type=::NewEnumerator1)
     1=&RandomRangeFloat(Minimum=0.05, Maximum=0.08, Evaluation Type=::NewEnumerator1)
     2=?
     3=?
     Index=&SpawnGroup()
     Num=::NewEnumerator0
    )
    Sprite Rotation Angle=&RandomRangeFloat(Minimum=0, Maximum=1, Evaluation Type=::NewEnumerator1)
    Sprite Rotation Mode=::NewEnumerator1
    Sprite Size=&MergeHelper_Vector2Switch(
     0=&RandomRangeVector2D(Minimum=<110 5>, Maximum=<150 90>, Evaluation Type=::NewEnumerator1)
     1=&Vector2DFromFloat(Value=&RandomRangeFloat(Minimum=25, Maximum=40, Evaluation Type=::NewEnumerator1))
     2=?
     3=?
     4=?
     Index=&SpawnGroup()
     Num=::NewEnumerator0
    )
    Sprite Size Mode=::NewEnumerator4
    UsePositionOffset=false
    Write Material Random=true
   )
   &CameraOffset(Camera Offset Amount=&RandomRangeFloat(Minimum=-2, Maximum=5, Evaluation Type=::NewEnumerator1))
   &AddVelocity(Velocity=&RandomRangeVector(Minimum=<0 0 -80>, Maximum=<0 0 -50>, Evaluation Type=::NewEnumerator1))
   /SetVariables(Particles.VisibilityTag=&SpawnGroup()) {isSetParameters=true}
  ]
  particleUpdateScript [
   &ParticleState()
   &SpriteRotationRate(Rotation Rate=&MergeHelper_FloatSwitch(
     0=&AngleConversion(
      Angle:NiagaraFloat=$Particles.SpriteRotationRate
      Angle Input:ENiagara_AngleInput=::NewEnumerator1
      Angle Output:ENiagara_AngleInput=::NewEnumerator0
     )
     1=0
     2=?
     3=?
     Index=&SpawnGroup()
     Num=::NewEnumerator0
    ))
   &ScaleSpriteSize(
    Scale Factor=&MergeHelper_Vector2Switch(
     0=&Vector2DFromFloat(Value=&FloatFromCurve(FloatCurve=DataInterface() {propertyValues {Curve {keys [
           <at -0.0011 0.0573>
           {arriveTangent=-0.0322, interpMode=RCIM_Cubic, leaveTangent=-0.0322, time=0.0867, value=0.7609}
           {interpMode=RCIM_Cubic, leaveTangent=0.0935, time=1.0043, value=0.0249}
          ]}}}))
     1=&Vector2DFromFloat(Value=&FloatFromCurve(FloatCurve=DataInterface() {propertyValues {Curve {keys [
           {arriveTangent=2.8238, interpMode=RCIM_Cubic, leaveTangent=2.8238, tangentMode=RCTM_User, time=0, value=0}
           {arriveTangentWeight=0.2609, interpMode=RCIM_Cubic, leaveTangentWeight=0.1118, tangentMode=RCTM_User, time=0.7, value=1}
           {arriveTangent=-2.438, interpMode=RCIM_Cubic, leaveTangent=-2.438, tangentMode=RCTM_User, time=1, value=0.5}
          ]}}}))
     2=?
     3=?
     4=?
     Index=&SpawnGroup()
     Num=::NewEnumerator0
    )
    Scale Sprite Size Mode=::NewEnumerator2
   )
   &SolveForcesAndVelocity(Clamp Velocity=false, Limit Acceleration=false)
   &ScaleColor(
    Scale Alpha=&MergeHelper_FloatSwitch(
     0=&FloatFromCurve(FloatCurve=DataInterface() {propertyValues {Curve {keys [<at 0 0 RCIM_Cubic>,<at 0.0967 0.1386 RCIM_Cubic>,<at 1 0>]}}})
     1=1
     2=?
     3=?
     Index=&SpawnGroup()
     Num=::NewEnumerator0
    )
    ScaleA=true
    ScaleRGB=true
   )
   /SetVariables(Particles.SubImageIndex=&MergeHelper_FloatSwitch(
     0=0
     1=&Lerp_Float(A=0, B=3, Alpha=$Particles.NormalizedAge)
     2=?
     3=?
     Index=&SpawnGroup()
     Num=::NewEnumerator0
    )) {isSetParameters=true}
  ]
  rendererClasses [&NiagaraSpriteRendererProperties]
  renderers [
   {rendererClass=&NiagaraSpriteRendererProperties}
   {rendererClass=&NiagaraSpriteRendererProperties, rendererIndex=1}
  ]
 }
}
summary {
 systemName=NS_GrillMoon_Jolly_Launch
 emitters [
  {emitterName=Glow, rendererClasses [&NiagaraSpriteRendererProperties, &NiagaraLightRendererProperties]}
  {bEnabled=false, emitterName=Rocket, rendererClasses [&NiagaraMeshRendererProperties]}
  {bEnabled=false, emitterName=fuse, rendererClasses [&NiagaraSpriteRendererProperties]}
  {emitterName=RiftShards, rendererClasses [&NiagaraMeshRendererProperties]}
  {bEnabled=false, emitterName=SparksClose, simTarget=GPUComputeSim, rendererClasses [&NiagaraSpriteRendererProperties]}
  {bEnabled=false, emitterName=TransSmoke, rendererClasses [&NiagaraSpriteRendererProperties]}
  {emitterName=Light, rendererClasses [&NiagaraLightRendererProperties]}
  {bEnabled=false, emitterName=RisingFlare, rendererClasses [&NiagaraSpriteRendererProperties]}
  {bEnabled=false, emitterName=ElectricArcsResidual, rendererClasses [&NiagaraMeshRendererProperties]}
  {emitterName=Lightning001, rendererClasses [&NiagaraMeshRendererProperties]}
 ]
 userVariables [
  User.FXColorA {type=LinearColor, description="", defaultValue {struct=LinearColor, value=<rgba 1 0.6933 0.0314 1>}}
  User.FXColorB {type=LinearColor, description="", defaultValue {struct=LinearColor, value=<rgba 1 0.55 0.1 1>}}
  User.FXColorC {type=LinearColor, description="", defaultValue {struct=LinearColor, value=<rgba 1 0.7313 0.38 1>}}
  User.FXColorD {type=LinearColor, description="", defaultValue {struct=LinearColor, value=<rgba 1 0.67 0.1 1>}}
 ]
}
